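# Niagara idle effect for the Galaxy Vampire Hunter body: a persistent glow on the character's hands.
# One burst spawns two long-lived sprites, each following its own skeletal mesh bone or socket with a mirrored offset.
# Three sprite renderers accompany separate glow, smoke and star color/size attributes, with a waveform pulsing size and smoke tint.
# A binary camera-distance test (ADSAlpha) gates all alphas, while time of day also scales glow and star opacity.
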NiagaraSystem /BRCosmetics/Characters/Player/Female/Medium/Bodies/F_MED_Vampire_Hunter/FX/NS_Vampire_Hunter_Galaxy_HandGlowIdle
dependencies {
 usedDataInterfaces [&NiagaraDataInterfaceCamera, &NiagaraDataInterfaceCurve, &NiagaraDataInterfaceSkeletalMesh]
 usedDynamicInputs [&Boolean_NotOperation]
 usedModules [
  &self:SystemSpawnScript.SystemScriptSource.SystemScriptGraph.NiagaraNodeAssignment_1.SetVariables_#1
  &SystemState
  &self:SystemSpawnScript.SystemScriptSource.SystemScriptGraph.NiagaraNodeAssignment_0.SetVariables_#2
 ]
}
emitters {Hand_Glow {
  emitterUpdateScript [
   &EmitterState(Life Cycle Mode=::NewEnumerator1, Loop Behavior=::NewEnumerator1, Loop Duration Mode=::NewEnumerator1, UseLoopDelay=false)
   &SpawnBurst_Instantaneous(Spawn Count=2, Use Spawn Probability=false, UseLoopCountLimit=true)
  ]
  particleSpawnScript [
   &InitializeParticle(
    Color=<rgba 0 1.3333 4 0.85>
    Sprite Rotation Angle=&RandomRangeFloat(Minimum=-1, Maximum=1)
    Sprite Rotation Mode=::NewEnumerator1
    Sprite Size Mode=::NewEnumerator3
    Uniform Sprite Size=30
    UsePositionOffset=false
    Write Material Random=true
   )
   &PartitionParticles()
   &CameraOffset(Camera Offset Amount=6)
   /SetVariables(Particles.CameraOffsetSmoke=10) {isSetParameters=true}
  ]
  particleUpdateScript [
   /SetVariables(Particles.ADSAlpha:NiagaraFloat=&MakeCustomFloatFromBool(
     Bool=&SetBoolByFloatComparison(
      A=&VectorLength(Vector=&Subtract_Vector(
        A=&ReturnCameraPosition(Coordinate Space:ENiagaraCoordinateSpace=::World, Camera Query:NiagaraDataInterfaceCamera=DataInterface())
        B=Linked() {linkedVariable {name=Particles.Position, type=NiagaraPosition}}
       ))
      B=80
     )
     True Float=1
     False Float=0
    )) {isSetParameters=true}
   &SkeletalMeshLocation(
    Bone / Socket Index=$Particles.PartitionParticles.Partition
    Bone Sampling Mode=::NewEnumerator4
    MaskSpawnGroup=false
    Sampled Position Offset=&MaskVectorBySpawnGroup(Masked Vector=<6 -3.4 0>, Passthrough Vector=<-6 3.4 0>, Spawn Group Mask=$Particles.PartitionParticles.Partition)
    Skeletal Mesh=$System.Skeletal Mesh
   )
   /Niagara/Noise/Waveform(EnableExponent1=false, GlobalClampMax=false, GlobalClampMin=false, UseInput1=false, UsePhase1=false, "[1] Frequency"=0.2) {module=Waveform001}
   /SetVariables(Particles.ColorSmoke:LinearColor=&MultiplyLinearColor(
     Linear Color=&LerpLinearColors(StartColor=<rgba 0.0531 0.0425 0.17 0.8>, EndColor=<rgba 0.0587 0.033 0.11 0.85>, LerpFactor=$Particles.Waveform001.Waveform)
     Scale Factor=&MakeFloat4FromVectorAndFloat("Vector (RGB)"=<1 1 1>, "Float (Alpha)"=$Particles.ADSAlpha)
    )) {isSetParameters=true}
   /SetVariables(Particles.ColorStars:LinearColor=&MultiplyLinearColor(
     Linear Color=<rgba 1 1 1 0.7>
     Scale Factor=&MakeFloat4FromVectorAndFloat(
      "Vector (RGB)"=<1 1 1>
      "Float (Alpha)"=&Multiply_Float(
       A=$Particles.ADSAlpha
       B=&FNTimeOfDayScalar(Night Scale=0.75, Morning Scale=0.9, Evening Scale=0.9)
      )
     )
    )) {isSetParameters=true}
   &ScaleColor(
    Scale Alpha=&Multiply_Float(
     A=$Particles.ADSAlpha
     B=&FNTimeOfDayScalar(Night Scale=0.4, Morning Scale=0.7, Day Scale=0.9, Evening Scale=0.8)
    )
    ScaleA=true
    ScaleRGB=false
   )
   &ParticleState(Kill Particles When Lifetime Has Elapsed=false, Let Infinitely Lived Particles Die When Emitter Deactivates=true)
   &ScaleSpriteSize(
    Scale Sprite Size Mode=::NewEnumerator0
    Uniform Scale Factor=&Lerp_Float(A=0.77, B=1, Alpha=$Particles.Waveform001.Waveform)
   )
   &SpriteRotationRate(Rotation Rate=&RandomRangeFloat(Minimum=15, Maximum=22))
   /SetVariables(Particles.SmokeSize:Vector2f=&Multiply_Vector2D_ByFloat(Float=1, Vector2D=$Particles.SpriteSize)) {isSetParameters=true}
   /SetVariables(Particles.StarsSize:Vector2f=&Multiply_Vector2D_ByFloat(Float=1.95, Vector2D=$Particles.SpriteSize)) {isSetParameters=true}
  ]
  rendererClasses [&NiagaraSpriteRendererProperties]
  renderers [
   {rendererClass=&NiagaraSpriteRendererProperties}
   {rendererClass=&NiagaraSpriteRendererProperties, rendererIndex=1}
   {rendererClass=&NiagaraSpriteRendererProperties, rendererIndex=2}
  ]
 }}
summary {
 systemName=NS_Vampire_Hunter_Galaxy_HandGlowIdle
 emitters [{emitterName=Hand_Glow, rendererClasses [&NiagaraSpriteRendererProperties]}]
 userVariables [
  User.bIsFrontEnd {type=NiagaraBool}
  User.bIsFrontEndPreview {type=NiagaraBool}
  User.bisHandHidden {type=NiagaraBool}
  User.bKillReaction {type=NiagaraBool}
  User.IsSkydiving {type=NiagaraBool}
 ]
}
