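# Physics asset for the crouching pose of the female medium Vampire Hunter body, set up for dynamic skirt bones.
# Nine front, side and back capsule chains hang from tiny MSTR anchor bodies angularly locked to spine_01, with angular limits tightened on some side segments.
# Pelvis, spine, thigh and calf bodies use free angular constraints and form the leg and hip collision shapes.
# Two root spheres, one very large and low, apparently approximate the ground.
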PhysicsAsset /BRCosmetics/Characters/Player/Female/Medium/Bodies/F_MED_Vampire_Hunter/Meshes/F_MED_Vampire_Hunter_Crouching_Physics
bodies {
 calf_l [{shapeType=Capsule, length=34.1166, radius=11.5294, shapeName=calf_l_capsule, center=<-28.1755 -3 0>, rotation=<rot 90 0 0>}]
 calf_r [{shapeType=Capsule, length=21.5169, radius=12.1602, shapeName=calf_r_capsule, center=<22.348 2.6301 0>, rotation=<rot 90 180 -175.3764>}]
 dyn_skirt_back_01_c [{shapeType=Capsule, length=3.5276, radius=2.4, shapeName=dyn_skirt_back_01_c_capsule, center=<9.8833 0 0>, rotation=<rot 90 0 0>}]
 dyn_skirt_back_01_c_MSTR [{shapeType=Capsule, length=0.5, radius=0.5, shapeName=dyn_skirt_back_01_c_MSTR_capsule}]
 dyn_skirt_back_01_l [{shapeType=Capsule, length=9.5276, radius=1.6, shapeName=dyn_skirt_back_01_l_capsule, center=<9.8833 0 0>, rotation=<rot 90 0 0>}]
 dyn_skirt_back_01_l_MSTR [{shapeType=Capsule, length=0.5, radius=0.5, shapeName=dyn_skirt_back_01_l_MSTR_capsule}]
 dyn_skirt_back_01_r [{shapeType=Capsule, length=9.5276, radius=1.6, shapeName=dyn_skirt_back_01_r_capsule, center=<-11.2697 0 0>, rotation=<rot 90 0 0>}]
 dyn_skirt_back_01_r_MSTR [{shapeType=Capsule, length=0.5, radius=0.5, shapeName=dyn_skirt_back_01_r_MSTR_capsule, rotation=<rot 90 0 0>}]
 dyn_skirt_back_02_c [{shapeType=Capsule, length=4.5, radius=4, shapeName=dyn_skirt_back_02_c_capsule, center=<6.8291 0 0>, rotation=<rot 90 0 0>}]
 dyn_skirt_back_02_l [{shapeType=Capsule, length=19.2755, radius=3.5807, shapeName=dyn_skirt_back_02_l_capsule, center=<6.8291 0 2.6134>, rotation=<rot 0 0 90>}]
 dyn_skirt_back_02_r [{shapeType=Capsule, length=20.7337, radius=4.7231, shapeName=dyn_skirt_back_02_r_capsule, center=<-7.6854 0 -3.0638>, rotation=<rot 0 0 90>}]
 dyn_skirt_back_03_c [{shapeType=Capsule, length=0, radius=5.25, shapeName=dyn_skirt_back_03_c_capsule, center=<6.8291 0 0>, rotation=<rot 90 0 0>}]
 dyn_skirt_back_03_l [{shapeType=Capsule, length=20.2845, radius=6.1655, shapeName=dyn_skirt_back_03_l_capsule, center=<6.8291 0 2.6134>, rotation=<rot 0 0 90>}]
 dyn_skirt_back_03_r [{shapeType=Capsule, length=23.9693, radius=6.7403, shapeName=dyn_skirt_back_03_r_capsule, center=<-7.6854 0 -3>, rotation=<rot 0 0 90>}]
 dyn_skirt_back_04_l [{shapeType=Capsule, length=8.1325, radius=5.8429, shapeName=dyn_skirt_back_04_l_capsule, center=<10.6392 0 0>, rotation=<rot 90 0 0>}]
 dyn_skirt_back_04_r [{shapeType=Capsule, length=9.8189, radius=5.7018, shapeName=dyn_skirt_back_04_r_capsule, center=<-10.4175 0 0>, rotation=<rot 90 0 0>}]
 dyn_skirt_front_01_l [{shapeType=Capsule, length=3.4359, radius=5.5641, shapeName=dyn_skirt_front_01_l_capsule, center=<13.1745 0 4.1081>, rotation=<rot 90 0 0>}]
 dyn_skirt_front_01_l_MSTR [{shapeType=Capsule, length=0.5, radius=0.5, shapeName=dyn_skirt_front_01_l_MSTR_capsule}]
 dyn_skirt_front_01_r [{shapeType=Capsule, length=7.761, radius=2.239, shapeName=dyn_skirt_front_01_r_capsule, center=<-6.257 0 -0.7211>, rotation=<rot 90 0 0>}]
 dyn_skirt_front_01_r_MSTR [{shapeType=Capsule, length=0.5, radius=0.5, shapeName=dyn_skirt_front_01_r_MSTR_capsule, rotation=<rot 90 0 0>}]
 dyn_skirt_front_02_l [{shapeType=Capsule, length=0.5764, radius=6.7542, shapeName=dyn_skirt_front_02_l_capsule, center=<6.8291 0 3.4084>, rotation=<rot 90 0 0>}]
 dyn_skirt_front_02_r [{shapeType=Capsule, length=5.761, radius=3.239, shapeName=dyn_skirt_front_02_r_capsule, center=<-6.4127 0 -0.3098>, rotation=<rot 90 0 0>}]
 dyn_skirt_front_03_l [{shapeType=Capsule, length=6.7827, radius=3.2173, shapeName=dyn_skirt_front_03_l_capsule, center=<6.8291 0 2.7084>, rotation=<rot 90 0 0>}]
 dyn_skirt_front_03_r [{shapeType=Capsule, length=6.761, radius=3.239, shapeName=dyn_skirt_front_03_r_capsule, center=<-6.8291 0 -0.6291>, rotation=<rot 90 0 0>}]
 dyn_skirt_front_04_l [{shapeType=Capsule, length=6.5327, radius=2.4673, shapeName=dyn_skirt_front_04_l_capsule, center=<5.1575 0 2.7084>, rotation=<rot 90 0 0>}]
 dyn_skirt_front_04_r [{shapeType=Capsule, length=5.511, radius=3.739, shapeName=dyn_skirt_front_04_r_capsule, center=<-6.6051 0 -0.1267>, rotation=<rot 90 0 0>}]
 dyn_skirt_sideA_01_l [{shapeType=Capsule, length=7.2988, radius=1.3277, shapeName=dyn_skirt_sideA_01_l_capsule, center=<6.8291 0 0.641>, rotation=<rot 90 0 0>}]
 dyn_skirt_sideA_01_l_MSTR [{shapeType=Capsule, length=0.5, radius=0.5, shapeName=dyn_skirt_sideA_01_l_MSTR_capsule}]
 dyn_skirt_sideA_01_r [{shapeType=Capsule, length=11.75, radius=2, shapeName=dyn_skirt_sideA_01_r_capsule, center=<-6.8291 0 0>, rotation=<rot 90 180 180>}]
 dyn_skirt_sideA_01_r_MSTR [{shapeType=Capsule, length=0.5, radius=0.5, shapeName=dyn_skirt_sideA_01_r_MSTR_capsule, rotation=<rot 90 0 0>}]
 dyn_skirt_sideA_02_l [{shapeType=Capsule, length=10.4691, radius=3.2809, shapeName=dyn_skirt_sideA_02_l_capsule, center=<6.8291 0 1.0115>, rotation=<rot 90 0 0>}]
 dyn_skirt_sideA_02_r [{shapeType=Capsule, length=9.75, radius=4, shapeName=dyn_skirt_sideA_02_r_capsule, center=<-6.8291 0 0>, rotation=<rot 90 180 180>}]
 dyn_skirt_sideA_03_l [{shapeType=Capsule, length=5.9691, radius=4.0309, shapeName=dyn_skirt_sideA_03_l_capsule, center=<6.8291 0 1.0115>, rotation=<rot 90 0 0>}]
 dyn_skirt_sideA_03_r [{shapeType=Capsule, length=15.1859, radius=8.666, shapeName=dyn_skirt_sideA_03_r_capsule, center=<-10.1499 0 0>, rotation=<rot 0 0 90>}]
 dyn_skirt_sideA_04_l [{
   shapeType=Capsule
   length=11.0412
   radius=4.7088
   shapeName=dyn_skirt_sideA_04_l_capsule
   center=<8.9353 -1.6802 0.3231>
   rotation=<rot -14.0233 -180 -101.2114>
  }]
 dyn_skirt_sideA_04_r [{shapeType=Capsule, length=18.1331, radius=7.5493, shapeName=dyn_skirt_sideA_04_r_capsule, center=<-9.2282 0 0>, rotation=<rot 0 0 90>}]
 dyn_skirt_sideB_01_l [{shapeType=Capsule, length=12, radius=1.75, shapeName=dyn_skirt_sideB_01_l_capsule, center=<6.8291 0 0>, rotation=<rot 90 0 0>}]
 dyn_skirt_sideB_01_l_MSTR [{shapeType=Capsule, length=0.5, radius=0.5, shapeName=dyn_skirt_sideB_01_l_MSTR_capsule}]
 dyn_skirt_sideB_01_r [{shapeType=Capsule, length=12, radius=1.75, shapeName=dyn_skirt_sideB_01_r_capsule, center=<-6.8291 0 0>, rotation=<rot 90 180 180>}]
 dyn_skirt_sideB_01_r_MSTR [{shapeType=Capsule, length=0.5, radius=0.5, shapeName=dyn_skirt_sideB_01_r_MSTR_capsule, rotation=<rot 90 0 0>}]
 dyn_skirt_sideB_02_l [{shapeType=Capsule, length=26.5773, radius=4.8969, shapeName=dyn_skirt_sideB_02_l_capsule, center=<6.8291 0 2.8643>, rotation=<rot 0 0 90>}]
 dyn_skirt_sideB_02_r [{shapeType=Capsule, length=24.7149, radius=5.0724, shapeName=dyn_skirt_sideB_02_r_capsule, center=<-6.8291 0 -3.5184>, rotation=<rot 0 0 90>}]
 dyn_skirt_sideB_03_l [{shapeType=Capsule, length=20.4865, radius=7.2377, shapeName=dyn_skirt_sideB_03_l_capsule, center=<6.8291 0 2.8643>, rotation=<rot 0 0 90>}]
 dyn_skirt_sideB_03_r [{
   shapeType=Capsule
   length=20.1949
   radius=10.4481
   shapeName=dyn_skirt_sideB_03_r_capsule
   center=<-10.1537 0 -3.5184>
   rotation=<rot 0 0 90>
  }]
 dyn_skirt_sideB_04_l [{shapeType=Capsule, length=27.5136, radius=9.9606, shapeName=dyn_skirt_sideB_04_l_capsule, center=<12.1137 0 2.8643>, rotation=<rot 0 0 90>}]
 dyn_skirt_sideB_04_r [{shapeType=Capsule, length=20.6259, radius=10.0171, shapeName=dyn_skirt_sideB_03_r_capsule, center=<-13.109 0 -3.5184>, rotation=<rot 0 0 90>}]
 pelvis [
  {shapeType=Capsule, length=8, radius=7.75, shapeName=pelvis_capsule}
  {shapeType=Capsule, length=5.25, radius=8.75, shapeName=pelvis_capsule, center=<0 -2.2098 0>}
 ]
 root [
  {shapeType=Capsule, length=0, radius=183.4375, shapeName=root_sphere, center=<0 -8.6403 -175.5073>, rotation=<rot 0 -90 0>}
  {shapeType=Capsule, length=0, radius=27.8343, shapeName=root_sphere, center=<0.0002 2.4787 8.5972>, rotation=<rot 0 -90 0>}
 ]
 spine_01 [{shapeType=Capsule, length=3, radius=8.25, shapeName=spine_01_capsule}]
 thigh_l [
  {shapeType=Capsule, length=34.8334, radius=8.4166, shapeName=thigh_l_capsule, center=<-19 0 1.25>, rotation=<rot 85 179.9996 -177.5091>}
  {shapeType=Capsule, length=0, radius=8, shapeName=thigh_l_capsule, center=<-3.1 -2.7 3.1>, rotation=<rot 90 0 0>}
  {
   shapeType=Capsule
   length=25
   radius=5
   shapeName=thigh_l_capsule
   center=<-24.6187 12.4003 4.4618>
   rotation=<rot 81.7406 179.9753 -179.8298>
  }
 ]
 thigh_r [
  {shapeType=Capsule, length=34.1076, radius=9.1424, shapeName=thigh_r_capsule, center=<19 0 -1.25>, rotation=<rot 95 0 0>}
  {shapeType=Capsule, length=0, radius=8, shapeName=thigh_r_capsule, center=<3.1 2.7 -3.1>}
  {
   shapeType=Capsule
   length=19.5004
   radius=5.4996
   shapeName=thigh_r_capsule
   center=<21.8387 -13.3442 -3.8165>
   rotation=<rot -80.299 -0.0001 0>
  }
 ]
}
constraints [
 {
  bone1Name=spine_01
  bone2Name=pelvis
  swing1LimitDegrees=45
  swing1Motion=Free
  swing2LimitDegrees=45
  swing2Motion=Free
  twistLimitDegrees=45
  twistMotion=Free
 }
 {
  bone1Name=dyn_skirt_sideA_01_l_MSTR
  bone2Name=spine_01
  swing1LimitDegrees=45
  swing1Motion=Locked
  swing2LimitDegrees=45
  swing2Motion=Locked
  twistLimitDegrees=45
  twistMotion=Locked
 }
 {bone1Name=dyn_skirt_sideA_01_l, bone2Name=dyn_skirt_sideA_01_l_MSTR, swing1LimitDegrees=45, swing2LimitDegrees=45, twistLimitDegrees=45}
 {bone1Name=dyn_skirt_sideA_02_l, bone2Name=dyn_skirt_sideA_01_l, swing1LimitDegrees=45, swing2LimitDegrees=45, twistLimitDegrees=45}
 {bone1Name=dyn_skirt_sideA_03_l, bone2Name=dyn_skirt_sideA_02_l, swing1LimitDegrees=45, swing2LimitDegrees=45, twistLimitDegrees=5}
 {bone1Name=dyn_skirt_sideA_04_l, bone2Name=dyn_skirt_sideA_03_l, swing1LimitDegrees=45, swing2LimitDegrees=45, twistLimitDegrees=5}
 {
  bone1Name=dyn_skirt_back_01_c_MSTR
  bone2Name=spine_01
  swing1LimitDegrees=45
  swing1Motion=Locked
  swing2LimitDegrees=45
  swing2Motion=Locked
  twistLimitDegrees=45
  twistMotion=Locked
 }
 {bone1Name=dyn_skirt_back_01_c, bone2Name=dyn_skirt_back_01_c_MSTR, swing1LimitDegrees=45, swing2LimitDegrees=45, twistLimitDegrees=45}
 {bone1Name=dyn_skirt_back_02_c, bone2Name=dyn_skirt_back_01_c, swing1LimitDegrees=45, swing2LimitDegrees=45, twistLimitDegrees=45}
 {bone1Name=dyn_skirt_back_03_c, bone2Name=dyn_skirt_back_02_c, swing1LimitDegrees=45, swing2LimitDegrees=45, twistLimitDegrees=45}
 {
  bone1Name=dyn_skirt_sideB_01_l_MSTR
  bone2Name=spine_01
  swing1LimitDegrees=45
  swing1Motion=Locked
  swing2LimitDegrees=45
  swing2Motion=Locked
  twistLimitDegrees=45
  twistMotion=Locked
 }
 {bone1Name=dyn_skirt_sideB_01_l, bone2Name=dyn_skirt_sideB_01_l_MSTR, swing1LimitDegrees=45, swing2LimitDegrees=45, twistLimitDegrees=45}
 {bone1Name=dyn_skirt_sideB_02_l, bone2Name=dyn_skirt_sideB_01_l, swing1LimitDegrees=45, swing2LimitDegrees=45, twistLimitDegrees=45}
 {bone1Name=dyn_skirt_sideB_03_l, bone2Name=dyn_skirt_sideB_02_l, swing1LimitDegrees=45, swing2LimitDegrees=45, twistLimitDegrees=45}
 {bone1Name=dyn_skirt_sideB_04_l, bone2Name=dyn_skirt_sideB_03_l, swing1LimitDegrees=45, swing2LimitDegrees=45, twistLimitDegrees=45}
 {
  bone1Name=dyn_skirt_front_01_l_MSTR
  bone2Name=spine_01
  swing1LimitDegrees=45
  swing1Motion=Locked
  swing2LimitDegrees=45
  swing2Motion=Locked
  twistLimitDegrees=45
  twistMotion=Locked
 }
 {bone1Name=dyn_skirt_front_01_l, bone2Name=dyn_skirt_front_01_l_MSTR, swing1LimitDegrees=45, swing2LimitDegrees=45, twistLimitDegrees=45}
 {bone1Name=dyn_skirt_front_02_l, bone2Name=dyn_skirt_front_01_l, swing1LimitDegrees=45, swing2LimitDegrees=45, twistLimitDegrees=45}
 {bone1Name=dyn_skirt_front_03_l, bone2Name=dyn_skirt_front_02_l, swing1LimitDegrees=45, swing2LimitDegrees=45, twistLimitDegrees=45}
 {bone1Name=dyn_skirt_front_04_l, bone2Name=dyn_skirt_front_03_l, swing1LimitDegrees=45, swing2LimitDegrees=45, twistLimitDegrees=45}
 {
  bone1Name=dyn_skirt_back_01_l_MSTR
  bone2Name=spine_01
  swing1LimitDegrees=45
  swing1Motion=Locked
  swing2LimitDegrees=45
  swing2Motion=Locked
  twistLimitDegrees=45
  twistMotion=Locked
 }
 {bone1Name=dyn_skirt_back_01_l, bone2Name=dyn_skirt_back_01_l_MSTR, swing1LimitDegrees=45, swing2LimitDegrees=45, twistLimitDegrees=45}
 {bone1Name=dyn_skirt_back_02_l, bone2Name=dyn_skirt_back_01_l, swing1LimitDegrees=45, swing2LimitDegrees=45, twistLimitDegrees=45}
 {bone1Name=dyn_skirt_back_03_l, bone2Name=dyn_skirt_back_02_l, swing1LimitDegrees=45, swing2LimitDegrees=45, twistLimitDegrees=45}
 {bone1Name=dyn_skirt_back_04_l, bone2Name=dyn_skirt_back_03_l, swing1LimitDegrees=45, swing2LimitDegrees=45, twistLimitDegrees=45}
 {
  bone1Name=dyn_skirt_sideA_01_r_MSTR
  bone2Name=spine_01
  swing1LimitDegrees=45
  swing1Motion=Locked
  swing2LimitDegrees=45
  swing2Motion=Locked
  twistLimitDegrees=45
  twistMotion=Locked
 }
 {bone1Name=dyn_skirt_sideA_01_r, bone2Name=dyn_skirt_sideA_01_r_MSTR, swing1LimitDegrees=45, swing2LimitDegrees=45, twistLimitDegrees=45}
 {bone1Name=dyn_skirt_sideA_02_r, bone2Name=dyn_skirt_sideA_01_r, swing1LimitDegrees=45, swing2LimitDegrees=45, twistLimitDegrees=45}
 {bone1Name=dyn_skirt_sideA_03_r, bone2Name=dyn_skirt_sideA_02_r, swing1LimitDegrees=45, swing2LimitDegrees=45, twistLimitDegrees=45}
 {bone1Name=dyn_skirt_sideA_04_r, bone2Name=dyn_skirt_sideA_03_r, swing1LimitDegrees=45, swing2LimitDegrees=45, twistLimitDegrees=45}
 {
  bone1Name=dyn_skirt_sideB_01_r_MSTR
  bone2Name=spine_01
  swing1LimitDegrees=45
  swing1Motion=Locked
  swing2LimitDegrees=45
  swing2Motion=Locked
  twistLimitDegrees=45
  twistMotion=Locked
 }
 {bone1Name=dyn_skirt_sideB_01_r, bone2Name=dyn_skirt_sideB_01_r_MSTR, swing1LimitDegrees=10, swing2LimitDegrees=45, twistLimitDegrees=5}
 {bone1Name=dyn_skirt_sideB_02_r, bone2Name=dyn_skirt_sideB_01_r, swing1LimitDegrees=10, swing2LimitDegrees=45, twistLimitDegrees=5}
 {bone1Name=dyn_skirt_sideB_03_r, bone2Name=dyn_skirt_sideB_02_r, swing1LimitDegrees=10, swing2LimitDegrees=45, twistLimitDegrees=5}
 {
  bone1Name=dyn_skirt_back_01_r_MSTR
  bone2Name=spine_01
  swing1LimitDegrees=45
  swing1Motion=Locked
  swing2LimitDegrees=45
  swing2Motion=Locked
  twistLimitDegrees=45
  twistMotion=Locked
 }
 {bone1Name=dyn_skirt_back_01_r, bone2Name=dyn_skirt_back_01_r_MSTR, swing1LimitDegrees=45, swing2LimitDegrees=45, twistLimitDegrees=45}
 {bone1Name=dyn_skirt_back_02_r, bone2Name=dyn_skirt_back_01_r, swing1LimitDegrees=45, swing2LimitDegrees=45, twistLimitDegrees=45}
 {bone1Name=dyn_skirt_back_03_r, bone2Name=dyn_skirt_back_02_r, swing1LimitDegrees=45, swing2LimitDegrees=45, twistLimitDegrees=45}
 {bone1Name=dyn_skirt_back_04_r, bone2Name=dyn_skirt_back_03_r, swing1LimitDegrees=45, swing2LimitDegrees=45, twistLimitDegrees=45}
 {
  bone1Name=dyn_skirt_front_01_r_MSTR
  bone2Name=spine_01
  swing1LimitDegrees=45
  swing1Motion=Locked
  swing2LimitDegrees=45
  swing2Motion=Locked
  twistLimitDegrees=45
  twistMotion=Locked
 }
 {bone1Name=dyn_skirt_front_01_r, bone2Name=dyn_skirt_front_01_r_MSTR, swing1LimitDegrees=45, swing2LimitDegrees=45, twistLimitDegrees=45}
 {bone1Name=dyn_skirt_front_02_r, bone2Name=dyn_skirt_front_01_r, swing1LimitDegrees=45, swing2LimitDegrees=45, twistLimitDegrees=45}
 {bone1Name=dyn_skirt_front_03_r, bone2Name=dyn_skirt_front_02_r, swing1LimitDegrees=45, swing2LimitDegrees=45, twistLimitDegrees=45}
 {bone1Name=dyn_skirt_front_04_r, bone2Name=dyn_skirt_front_03_r, swing1LimitDegrees=45, swing2LimitDegrees=45, twistLimitDegrees=45}
 {
  bone1Name=thigh_l
  bone2Name=pelvis
  swing1LimitDegrees=45
  swing1Motion=Free
  swing2LimitDegrees=45
  swing2Motion=Free
  twistLimitDegrees=45
  twistMotion=Free
 }
 {
  bone1Name=calf_l
  bone2Name=thigh_l
  swing1LimitDegrees=45
  swing1Motion=Free
  swing2LimitDegrees=45
  swing2Motion=Free
  twistLimitDegrees=45
  twistMotion=Free
 }
 {
  bone1Name=thigh_r
  bone2Name=pelvis
  swing1LimitDegrees=45
  swing1Motion=Free
  swing2LimitDegrees=45
  swing2Motion=Free
  twistLimitDegrees=45
  twistMotion=Free
 }
 {
  bone1Name=calf_r
  bone2Name=thigh_r
  swing1LimitDegrees=45
  swing1Motion=Free
  swing2LimitDegrees=45
  swing2Motion=Free
  twistLimitDegrees=45
  twistMotion=Free
 }
 {
  bone1Name=pelvis
  bone2Name=root
  swing1LimitDegrees=45
  swing1Motion=Free
  swing2LimitDegrees=45
  swing2Motion=Free
  twistLimitDegrees=45
  twistMotion=Free
 }
 {bone1Name=dyn_skirt_sideB_04_r, bone2Name=dyn_skirt_sideB_03_r, swing1LimitDegrees=10, swing2LimitDegrees=45, twistLimitDegrees=5}
]
