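# Physics asset for the skydiving skirt of the female medium Vampire Hunter cosmetic body, built from capsules on skirt and lower-body bones.
# Nine skirt chains (front, two side sets, back) hang from MSTR anchors locked to spine_01, with limited per-segment swing and zero-degree twist limits.
# Pelvis, spine, legs and a huge offset root sphere provide body shapes joined by free angular constraints.
# A free angular constraint also cross-links the left and right back_03 segments.
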PhysicsAsset /BRCosmetics/Characters/Player/Female/Medium/Bodies/F_MED_Vampire_Hunter/Meshes/F_MED_Vampire_Hunter_Skydiving_Physics
bodies {
 calf_l [{shapeType=Capsule, length=38.896, radius=6.75, shapeName=calf_l_capsule, center=<-28.1755 -3 0>, rotation=<rot 90 0 0>}]
 calf_r [{shapeType=Capsule, length=26.9271, radius=6.75, shapeName=calf_r_capsule, center=<22.348 2.6301 0>, rotation=<rot 90 180 -175.3764>}]
 dyn_skirt_back_01_c [{shapeType=Capsule, length=6.25, radius=2.4, shapeName=dyn_skirt_back_01_c_capsule, center=<6.8291 0 0>, rotation=<rot 90 0 0>}]
 dyn_skirt_back_01_c_MSTR [{shapeType=Capsule, length=0.5, radius=0.5, shapeName=dyn_skirt_back_01_c_MSTR_capsule}]
 dyn_skirt_back_01_l [{shapeType=Capsule, length=12.25, radius=1.6, shapeName=dyn_skirt_back_01_l_capsule, center=<6.8291 0 0>, rotation=<rot 90 0 0>}]
 dyn_skirt_back_01_l_MSTR [{shapeType=Capsule, length=0.5, radius=0.5, shapeName=dyn_skirt_back_01_l_MSTR_capsule}]
 dyn_skirt_back_01_r [{shapeType=Capsule, length=12.25, radius=1.6, shapeName=dyn_skirt_back_01_r_capsule, center=<-6.8291 0 0>, rotation=<rot 90 180 180>}]
 dyn_skirt_back_01_r_MSTR [{shapeType=Capsule, length=0.5, radius=0.5, shapeName=dyn_skirt_back_01_r_MSTR_capsule, rotation=<rot 90 0 0>}]
 dyn_skirt_back_02_c [{shapeType=Capsule, length=4.5, radius=4, shapeName=dyn_skirt_back_02_c_capsule, center=<6.8291 0 0>, rotation=<rot 90 0 0>}]
 dyn_skirt_back_02_l [{shapeType=Capsule, length=9.75, radius=4, shapeName=dyn_skirt_back_02_l_capsule, center=<6.8291 0 0>, rotation=<rot 90 0 0>}]
 dyn_skirt_back_02_r [{shapeType=Capsule, length=9.75, radius=4, shapeName=dyn_skirt_back_02_r_capsule, center=<-6.8291 0 0>, rotation=<rot 90 180 180>}]
 dyn_skirt_back_03_c [{shapeType=Capsule, length=0, radius=5.25, shapeName=dyn_skirt_back_03_c_capsule, center=<6.8291 0 0>, rotation=<rot 90 0 0>}]
 dyn_skirt_back_03_l [{shapeType=Capsule, length=5.3031, radius=4.6969, shapeName=dyn_skirt_back_03_l_capsule, center=<6.8291 0 0>, rotation=<rot 90 0 0>}]
 dyn_skirt_back_03_r [{shapeType=Capsule, length=4.75, radius=5.25, shapeName=dyn_skirt_back_03_r_capsule, center=<-6.8291 0 0>, rotation=<rot 90 180 180>}]
 dyn_skirt_back_04_l [{shapeType=Capsule, length=9.25, radius=4.5, shapeName=dyn_skirt_back_04_l_capsule, center=<6.8291 0 0>, rotation=<rot 90 0 0>}]
 dyn_skirt_back_04_r [{shapeType=Capsule, length=9.25, radius=4.5, shapeName=dyn_skirt_back_04_r_capsule, center=<-6.8291 0 0>, rotation=<rot 90 180 180>}]
 dyn_skirt_front_01_l [{shapeType=Capsule, length=7.25, radius=1.75, shapeName=dyn_skirt_front_01_l_capsule, center=<6.3177 0 0.7>, rotation=<rot 90 0 0>}]
 dyn_skirt_front_01_l_MSTR [{shapeType=Capsule, length=0.5, radius=0.5, shapeName=dyn_skirt_front_01_l_MSTR_capsule}]
 dyn_skirt_front_01_r [{shapeType=Capsule, length=8.25, radius=1.75, shapeName=dyn_skirt_front_01_r_capsule, center=<-5.7786 0 0>, rotation=<rot 90 0 0>}]
 dyn_skirt_front_01_r_MSTR [{shapeType=Capsule, length=0.5, radius=0.5, shapeName=dyn_skirt_front_01_r_MSTR_capsule, rotation=<rot 90 0 0>}]
 dyn_skirt_front_02_l [{shapeType=Capsule, length=6, radius=3, shapeName=dyn_skirt_front_02_l_capsule, center=<6.8291 0 0.7>, rotation=<rot 90 0 0>}]
 dyn_skirt_front_02_r [{shapeType=Capsule, length=6.25, radius=2.75, shapeName=dyn_skirt_front_02_r_capsule, center=<-6.8291 0 0>, rotation=<rot 90 180 180>}]
 dyn_skirt_front_03_l [{shapeType=Capsule, length=6.5, radius=3.5, shapeName=dyn_skirt_front_03_l_capsule, center=<6.8291 0 0>, rotation=<rot 90 0 0>}]
 dyn_skirt_front_03_r [{shapeType=Capsule, length=7.25, radius=2.75, shapeName=dyn_skirt_front_03_r_capsule, center=<-6.8291 0 0>, rotation=<rot 90 180 180>}]
 dyn_skirt_front_04_l [{shapeType=Capsule, length=6.25, radius=2.75, shapeName=dyn_skirt_front_04_l_capsule, center=<5.1575 0 0>, rotation=<rot 90 0 0>}]
 dyn_skirt_front_04_r [{shapeType=Capsule, length=6, radius=3.25, shapeName=dyn_skirt_front_04_r_capsule, center=<-5.5437 0 0>, rotation=<rot 90 0 0>}]
 dyn_skirt_sideA_01_l [{shapeType=Capsule, length=11.75, radius=2, shapeName=dyn_skirt_sideA_01_l_capsule, center=<6.8291 0 0>, rotation=<rot 90 0 0>}]
 dyn_skirt_sideA_01_l_MSTR [{shapeType=Capsule, length=0.5, radius=0.5, shapeName=dyn_skirt_sideA_01_l_MSTR_capsule}]
 dyn_skirt_sideA_01_r [{shapeType=Capsule, length=11.75, radius=2, shapeName=dyn_skirt_sideA_01_r_capsule, center=<-6.8291 0 0>, rotation=<rot 90 180 180>}]
 dyn_skirt_sideA_01_r_MSTR [{shapeType=Capsule, length=0.5, radius=0.5, shapeName=dyn_skirt_sideA_01_r_MSTR_capsule, rotation=<rot 90 0 0>}]
 dyn_skirt_sideA_02_l [{shapeType=Capsule, length=9.25, radius=4.5, shapeName=dyn_skirt_sideA_02_l_capsule, center=<6.8291 0 0>, rotation=<rot 90 0 0>}]
 dyn_skirt_sideA_02_r [{shapeType=Capsule, length=9.75, radius=4, shapeName=dyn_skirt_sideA_02_r_capsule, center=<-6.8291 0 0>, rotation=<rot 90 180 180>}]
 dyn_skirt_sideA_03_l [{shapeType=Capsule, length=4.75, radius=5.25, shapeName=dyn_skirt_sideA_03_l_capsule, center=<6.8291 0 0>, rotation=<rot 90 0 0>}]
 dyn_skirt_sideA_03_r [{shapeType=Capsule, length=9.367, radius=4, shapeName=dyn_skirt_sideA_03_r_capsule, center=<-10.1499 0 0>, rotation=<rot 90 0 0>}]
 dyn_skirt_sideA_04_l [{shapeType=Capsule, length=12.5, radius=3.25, shapeName=dyn_skirt_sideA_04_l_capsule, center=<9.016 0 0>, rotation=<rot 90 0 0>}]
 dyn_skirt_sideA_04_r [{shapeType=Capsule, length=12.3143, radius=2.8833, shapeName=dyn_skirt_sideA_04_r_capsule, center=<-9.2282 0 0>, rotation=<rot 90 0 0>}]
 dyn_skirt_sideB_01_l [{shapeType=Capsule, length=12, radius=1.75, shapeName=dyn_skirt_sideB_01_l_capsule, center=<6.8291 0 0>, rotation=<rot 90 0 0>}]
 dyn_skirt_sideB_01_l_MSTR [{shapeType=Capsule, length=0.5, radius=0.5, shapeName=dyn_skirt_sideB_01_l_MSTR_capsule}]
 dyn_skirt_sideB_01_r [{shapeType=Capsule, length=12, radius=1.75, shapeName=dyn_skirt_sideB_01_r_capsule, center=<-6.8291 0 0>, rotation=<rot 90 180 180>}]
 dyn_skirt_sideB_01_r_MSTR [{shapeType=Capsule, length=0.5, radius=0.5, shapeName=dyn_skirt_sideB_01_r_MSTR_capsule, rotation=<rot 90 0 0>}]
 dyn_skirt_sideB_02_l [{shapeType=Capsule, length=9.75, radius=4, shapeName=dyn_skirt_sideB_02_l_capsule, center=<6.8291 0 0>, rotation=<rot 90 0 0>}]
 dyn_skirt_sideB_02_r [{shapeType=Capsule, length=9.75, radius=4, shapeName=dyn_skirt_sideB_02_r_capsule, center=<-6.8291 0 0>, rotation=<rot 90 180 180>}]
 dyn_skirt_sideB_03_l [{shapeType=Capsule, length=4.75, radius=5.25, shapeName=dyn_skirt_sideB_03_l_capsule, center=<6.8291 0 0>, rotation=<rot 90 0 0>}]
 dyn_skirt_sideB_03_r [{shapeType=Capsule, length=7.8215, radius=6.7843, shapeName=dyn_skirt_sideB_03_r_capsule, center=<-10.1537 0 0>, rotation=<rot 90 0 0>}]
 dyn_skirt_sideB_04_l [{shapeType=Capsule, length=15.25, radius=4.5, shapeName=dyn_skirt_sideB_04_l_capsule, center=<12.1137 0 0>, rotation=<rot 90 0 0>}]
 dyn_skirt_sideB_04_r [{
   shapeType=Capsule
   length=11.6423
   radius=4.5156
   shapeName=dyn_skirt_sideB_04_r_capsule
   center=<-14.2122 2.0816 0.7254>
   rotation=<rot 0 -101.882 -85.8459>
  }]
 pelvis [
  {shapeType=Capsule, length=8, radius=7.75, shapeName=pelvis_capsule}
  {shapeType=Capsule, length=5.25, radius=8.75, shapeName=pelvis_capsule, center=<0 -2.2098 0>}
 ]
 root [{shapeType=Capsule, length=0, radius=250, shapeName=root_capsule, center=<0 -24.9837 -246.8085>, rotation=<rot 0 -90 0>}]
 spine_01 [{shapeType=Capsule, length=3, radius=8.25, shapeName=spine_01_capsule}]
 thigh_l [
  {shapeType=Capsule, length=36.25, radius=7, shapeName=thigh_l_capsule, center=<-19 0 1.25>, rotation=<rot 85 179.9996 -177.5091>}
  {shapeType=Capsule, length=0, radius=8, shapeName=thigh_l_capsule, center=<-3.1 -2.7 3.1>, rotation=<rot 90 0 0>}
  {
   shapeType=Capsule
   length=25
   radius=5
   shapeName=thigh_l_capsule
   center=<-23.8859 12.3988 2.1683>
   rotation=<rot 72.2803 179.8835 -179.9197>
  }
 ]
 thigh_r [
  {shapeType=Capsule, length=36.25, radius=7, shapeName=thigh_r_capsule, center=<19 0 -1.25>, rotation=<rot 95 0 0>}
  {shapeType=Capsule, length=0, radius=8, shapeName=thigh_r_capsule, center=<3.1 2.7 -3.1>}
  {shapeType=Capsule, length=20.5, radius=4.5, shapeName=thigh_r_capsule, center=<21.8387 -12.2366 -3.8165>, rotation=<rot -80.299 -0.0001 0>}
 ]
}
constraints [
 {
  bone1Name=spine_01
  bone2Name=pelvis
  swing1LimitDegrees=45
  swing1Motion=Free
  swing2LimitDegrees=45
  swing2Motion=Free
  twistLimitDegrees=45
  twistMotion=Free
 }
 {
  bone1Name=dyn_skirt_sideA_01_l_MSTR
  bone2Name=spine_01
  swing1LimitDegrees=45
  swing1Motion=Locked
  swing2LimitDegrees=45
  swing2Motion=Locked
  twistLimitDegrees=45
  twistMotion=Locked
 }
 {bone1Name=dyn_skirt_sideA_01_l, bone2Name=dyn_skirt_sideA_01_l_MSTR, swing1LimitDegrees=5, swing2LimitDegrees=15, twistLimitDegrees=0}
 {bone1Name=dyn_skirt_sideA_02_l, bone2Name=dyn_skirt_sideA_01_l, swing1LimitDegrees=5, swing2LimitDegrees=20, twistLimitDegrees=0}
 {bone1Name=dyn_skirt_sideA_03_l, bone2Name=dyn_skirt_sideA_02_l, swing1LimitDegrees=5, swing2LimitDegrees=20, twistLimitDegrees=0}
 {bone1Name=dyn_skirt_sideA_04_l, bone2Name=dyn_skirt_sideA_03_l, swing1LimitDegrees=5, swing2LimitDegrees=20, twistLimitDegrees=0}
 {
  bone1Name=dyn_skirt_back_01_c_MSTR
  bone2Name=spine_01
  swing1LimitDegrees=45
  swing1Motion=Locked
  swing2LimitDegrees=45
  swing2Motion=Locked
  twistLimitDegrees=45
  twistMotion=Locked
 }
 {bone1Name=dyn_skirt_back_01_c, bone2Name=dyn_skirt_back_01_c_MSTR, swing1LimitDegrees=5, swing2LimitDegrees=15, twistLimitDegrees=0}
 {bone1Name=dyn_skirt_back_02_c, bone2Name=dyn_skirt_back_01_c, swing1LimitDegrees=5, swing2LimitDegrees=20, twistLimitDegrees=0}
 {bone1Name=dyn_skirt_back_03_c, bone2Name=dyn_skirt_back_02_c, swing1LimitDegrees=5, swing2LimitDegrees=20, twistLimitDegrees=0}
 {
  bone1Name=dyn_skirt_sideB_01_l_MSTR
  bone2Name=spine_01
  swing1LimitDegrees=45
  swing1Motion=Locked
  swing2LimitDegrees=45
  swing2Motion=Locked
  twistLimitDegrees=45
  twistMotion=Locked
 }
 {bone1Name=dyn_skirt_sideB_01_l, bone2Name=dyn_skirt_sideB_01_l_MSTR, swing1LimitDegrees=5, swing2LimitDegrees=15, twistLimitDegrees=0}
 {bone1Name=dyn_skirt_sideB_02_l, bone2Name=dyn_skirt_sideB_01_l, swing1LimitDegrees=5, swing2LimitDegrees=20, twistLimitDegrees=0}
 {bone1Name=dyn_skirt_sideB_03_l, bone2Name=dyn_skirt_sideB_02_l, swing1LimitDegrees=5, swing2LimitDegrees=20, twistLimitDegrees=0}
 {bone1Name=dyn_skirt_sideB_04_l, bone2Name=dyn_skirt_sideB_03_l, swing1LimitDegrees=5, swing2LimitDegrees=20, twistLimitDegrees=0}
 {
  bone1Name=dyn_skirt_front_01_l_MSTR
  bone2Name=spine_01
  swing1LimitDegrees=45
  swing1Motion=Locked
  swing2LimitDegrees=45
  swing2Motion=Locked
  twistLimitDegrees=45
  twistMotion=Locked
 }
 {bone1Name=dyn_skirt_front_01_l, bone2Name=dyn_skirt_front_01_l_MSTR, swing1LimitDegrees=45, swing2LimitDegrees=45, twistLimitDegrees=0}
 {bone1Name=dyn_skirt_front_02_l, bone2Name=dyn_skirt_front_01_l, swing1LimitDegrees=15, swing2LimitDegrees=15, twistLimitDegrees=0}
 {bone1Name=dyn_skirt_front_03_l, bone2Name=dyn_skirt_front_02_l, swing1LimitDegrees=15, swing2LimitDegrees=15, twistLimitDegrees=0}
 {bone1Name=dyn_skirt_front_04_l, bone2Name=dyn_skirt_front_03_l, swing1LimitDegrees=30, swing2LimitDegrees=30, twistLimitDegrees=0}
 {
  bone1Name=dyn_skirt_back_01_l_MSTR
  bone2Name=spine_01
  swing1LimitDegrees=45
  swing1Motion=Locked
  swing2LimitDegrees=45
  swing2Motion=Locked
  twistLimitDegrees=45
  twistMotion=Locked
 }
 {bone1Name=dyn_skirt_back_01_l, bone2Name=dyn_skirt_back_01_l_MSTR, swing1LimitDegrees=5, swing2LimitDegrees=15, twistLimitDegrees=0}
 {bone1Name=dyn_skirt_back_02_l, bone2Name=dyn_skirt_back_01_l, swing1LimitDegrees=10, swing2LimitDegrees=20, twistLimitDegrees=0}
 {bone1Name=dyn_skirt_back_03_l, bone2Name=dyn_skirt_back_02_l, swing1LimitDegrees=20, swing2LimitDegrees=30, twistLimitDegrees=0}
 {bone1Name=dyn_skirt_back_04_l, bone2Name=dyn_skirt_back_03_l, swing1LimitDegrees=20, swing2LimitDegrees=30, twistLimitDegrees=0}
 {
  bone1Name=dyn_skirt_sideA_01_r_MSTR
  bone2Name=spine_01
  swing1LimitDegrees=45
  swing1Motion=Locked
  swing2LimitDegrees=45
  swing2Motion=Locked
  twistLimitDegrees=45
  twistMotion=Locked
 }
 {bone1Name=dyn_skirt_sideA_01_r, bone2Name=dyn_skirt_sideA_01_r_MSTR, swing1LimitDegrees=5, swing2LimitDegrees=15, twistLimitDegrees=0}
 {bone1Name=dyn_skirt_sideA_02_r, bone2Name=dyn_skirt_sideA_01_r, swing1LimitDegrees=5, swing2LimitDegrees=20, twistLimitDegrees=0}
 {bone1Name=dyn_skirt_sideA_03_r, bone2Name=dyn_skirt_sideA_02_r, swing1LimitDegrees=5, swing2LimitDegrees=20, twistLimitDegrees=0}
 {bone1Name=dyn_skirt_sideA_04_r, bone2Name=dyn_skirt_sideA_03_r, swing1LimitDegrees=5, swing2LimitDegrees=20, twistLimitDegrees=0}
 {
  bone1Name=dyn_skirt_sideB_01_r_MSTR
  bone2Name=spine_01
  swing1LimitDegrees=45
  swing1Motion=Locked
  swing2LimitDegrees=45
  swing2Motion=Locked
  twistLimitDegrees=45
  twistMotion=Locked
 }
 {bone1Name=dyn_skirt_sideB_01_r, bone2Name=dyn_skirt_sideB_01_r_MSTR, swing1LimitDegrees=5, swing2LimitDegrees=15, twistLimitDegrees=0}
 {bone1Name=dyn_skirt_sideB_02_r, bone2Name=dyn_skirt_sideB_01_r, swing1LimitDegrees=10, swing2LimitDegrees=20, twistLimitDegrees=0}
 {bone1Name=dyn_skirt_sideB_03_r, bone2Name=dyn_skirt_sideB_02_r, swing1LimitDegrees=30, swing2LimitDegrees=30, twistLimitDegrees=0}
 {
  bone1Name=dyn_skirt_back_01_r_MSTR
  bone2Name=spine_01
  swing1LimitDegrees=45
  swing1Motion=Locked
  swing2LimitDegrees=45
  swing2Motion=Locked
  twistLimitDegrees=45
  twistMotion=Locked
 }
 {bone1Name=dyn_skirt_back_01_r, bone2Name=dyn_skirt_back_01_r_MSTR, swing1LimitDegrees=5, swing2LimitDegrees=15, twistLimitDegrees=0}
 {bone1Name=dyn_skirt_back_02_r, bone2Name=dyn_skirt_back_01_r, swing1LimitDegrees=10, swing2LimitDegrees=20, twistLimitDegrees=0}
 {bone1Name=dyn_skirt_back_03_r, bone2Name=dyn_skirt_back_02_r, swing1LimitDegrees=20, swing2LimitDegrees=30, twistLimitDegrees=0}
 {bone1Name=dyn_skirt_back_04_r, bone2Name=dyn_skirt_back_03_r, swing1LimitDegrees=20, swing2LimitDegrees=30, twistLimitDegrees=0}
 {
  bone1Name=dyn_skirt_front_01_r_MSTR
  bone2Name=spine_01
  swing1LimitDegrees=45
  swing1Motion=Locked
  swing2LimitDegrees=45
  swing2Motion=Locked
  twistLimitDegrees=45
  twistMotion=Locked
 }
 {bone1Name=dyn_skirt_front_01_r, bone2Name=dyn_skirt_front_01_r_MSTR, swing1LimitDegrees=30, swing2LimitDegrees=45, twistLimitDegrees=0}
 {bone1Name=dyn_skirt_front_02_r, bone2Name=dyn_skirt_front_01_r, swing1LimitDegrees=15, swing2LimitDegrees=15, twistLimitDegrees=0}
 {bone1Name=dyn_skirt_front_03_r, bone2Name=dyn_skirt_front_02_r, swing1LimitDegrees=10, swing2LimitDegrees=15, twistLimitDegrees=0}
 {bone1Name=dyn_skirt_front_04_r, bone2Name=dyn_skirt_front_03_r, swing1LimitDegrees=5, swing2LimitDegrees=20, twistLimitDegrees=0}
 {
  bone1Name=thigh_l
  bone2Name=pelvis
  swing1LimitDegrees=45
  swing1Motion=Free
  swing2LimitDegrees=45
  swing2Motion=Free
  twistLimitDegrees=45
  twistMotion=Free
 }
 {
  bone1Name=calf_l
  bone2Name=thigh_l
  swing1LimitDegrees=45
  swing1Motion=Free
  swing2LimitDegrees=45
  swing2Motion=Free
  twistLimitDegrees=45
  twistMotion=Free
 }
 {
  bone1Name=thigh_r
  bone2Name=pelvis
  swing1LimitDegrees=45
  swing1Motion=Free
  swing2LimitDegrees=45
  swing2Motion=Free
  twistLimitDegrees=45
  twistMotion=Free
 }
 {
  bone1Name=calf_r
  bone2Name=thigh_r
  swing1LimitDegrees=45
  swing1Motion=Free
  swing2LimitDegrees=45
  swing2Motion=Free
  twistLimitDegrees=45
  twistMotion=Free
 }
 {
  bone1Name=pelvis
  bone2Name=root
  swing1LimitDegrees=45
  swing1Motion=Free
  swing2LimitDegrees=45
  swing2Motion=Free
  twistLimitDegrees=45
  twistMotion=Free
 }
 {
  bone1Name=dyn_skirt_back_03_l
  bone2Name=dyn_skirt_back_03_r
  swing1LimitDegrees=45
  swing1Motion=Free
  swing2LimitDegrees=45
  swing2Motion=Free
  twistLimitDegrees=45
  twistMotion=Free
 }
 {bone1Name=dyn_skirt_sideB_04_r, bone2Name=dyn_skirt_sideB_03_r, swing1LimitDegrees=30, swing2LimitDegrees=30, twistLimitDegrees=0}
]
