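# Physics asset for the standing mesh of the female medium Vampire Hunter cosmetic body, set up for dynamic skirt motion.
# Front, side and back skirt chains use capsules joined by angle-limited constraints, anchored through tiny MSTR bodies rotation-locked to spine_01.
# Pelvis, spine, thighs and calves supply body shapes, while root carries two large zero-length-capsule spheres.
# Three free-rotation cross-constraints tie front panels to neighboring sideA panels.
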PhysicsAsset /BRCosmetics/Characters/Player/Female/Medium/Bodies/F_MED_Vampire_Hunter/Meshes/F_MED_Vampire_Hunter_Standing_Physics
bodies {
 calf_l [{shapeType=Capsule, length=36.152, radius=9.494, shapeName=calf_l_capsule, center=<-28.1755 -3 0>, rotation=<rot 90 0 0>}]
 calf_r [{shapeType=Capsule, length=24.1831, radius=9.494, shapeName=calf_r_capsule, center=<22.348 2.6301 0>, rotation=<rot 90 180 -175.3764>}]
 dyn_skirt_back_01_c [{shapeType=Capsule, length=8.5785, radius=3.0351, shapeName=dyn_skirt_back_01_c_capsule, center=<3.8261 0 2>, rotation=<rot 0 0 90>}]
 dyn_skirt_back_01_c_MSTR [{shapeType=Capsule, length=0.5, radius=0.5, shapeName=dyn_skirt_back_01_c_MSTR_capsule}]
 dyn_skirt_back_01_l [{shapeType=Capsule, length=7.0154, radius=4.1123, shapeName=dyn_skirt_back_01_l_capsule, center=<9.8833 0 3>, rotation=<rot 90 0 0>}]
 dyn_skirt_back_01_l_MSTR [{shapeType=Capsule, length=0.5, radius=0.5, shapeName=dyn_skirt_back_01_l_MSTR_capsule}]
 dyn_skirt_back_01_r [{shapeType=Capsule, length=6.4065, radius=4.7211, shapeName=dyn_skirt_back_01_r_capsule, center=<-11.2697 0 -3>, rotation=<rot 90 0 0>}]
 dyn_skirt_back_01_r_MSTR [{shapeType=Capsule, length=0.5, radius=0.5, shapeName=dyn_skirt_back_01_r_MSTR_capsule, rotation=<rot 90 0 0>}]
 dyn_skirt_back_02_c [{shapeType=Capsule, length=9.5509, radius=4.6351, shapeName=dyn_skirt_back_02_c_capsule, center=<0 0 2>, rotation=<rot 0 0 90>}]
 dyn_skirt_back_02_l [{shapeType=Capsule, length=17.0838, radius=3.5807, shapeName=dyn_skirt_back_02_l_capsule, center=<6.8291 0 2.6134>, rotation=<rot 0 0 90>}]
 dyn_skirt_back_02_r [{shapeType=Capsule, length=9.1357, radius=4.7231, shapeName=dyn_skirt_back_02_r_capsule, center=<-7.6854 0 -3.0638>, rotation=<rot 0 0 90>}]
 dyn_skirt_back_03_c [{shapeType=Capsule, length=5.0509, radius=5.8851, shapeName=dyn_skirt_back_03_c_capsule, center=<0 0 2>, rotation=<rot 0 0 90>}]
 dyn_skirt_back_03_l [{shapeType=Capsule, length=12.0459, radius=4, shapeName=dyn_skirt_back_03_l_capsule, center=<6.8291 0 0>, rotation=<rot 0 0 90>}]
 dyn_skirt_back_03_r [{shapeType=Capsule, length=12.0459, radius=4, shapeName=dyn_skirt_back_03_r_capsule, center=<-7.6854 0 0>, rotation=<rot 0 0 90>}]
 dyn_skirt_back_04_l [{shapeType=Capsule, length=12.0459, radius=4, shapeName=dyn_skirt_back_04_l_capsule, center=<6.8291 0 0>, rotation=<rot 0 0 90>}]
 dyn_skirt_back_04_r [{shapeType=Capsule, length=12.0459, radius=4, shapeName=dyn_skirt_back_04_r_capsule, center=<-7.6854 0 0>, rotation=<rot 0 0 90>}]
 dyn_skirt_front_01_l [{shapeType=Capsule, length=4.2355, radius=4.7645, shapeName=dyn_skirt_front_01_l_capsule, center=<13.1745 0 4.1081>, rotation=<rot 90 0 0>}]
 dyn_skirt_front_01_l_MSTR [{shapeType=Capsule, length=0.5, radius=0.5, shapeName=dyn_skirt_front_01_l_MSTR_capsule}]
 dyn_skirt_front_01_r [{shapeType=Capsule, length=5.0771, radius=4.8229, shapeName=dyn_skirt_front_01_r_capsule, center=<-5.7 0 -3>, rotation=<rot 0 0 90>}]
 dyn_skirt_front_01_r_MSTR [{shapeType=Capsule, length=0.5, radius=0.5, shapeName=dyn_skirt_front_01_r_MSTR_capsule, rotation=<rot 90 0 0>}]
 dyn_skirt_front_02_l [{shapeType=Capsule, length=1.5123, radius=5.8182, shapeName=dyn_skirt_front_02_l_capsule, center=<6.8291 0 3.4084>, rotation=<rot 90 0 0>}]
 dyn_skirt_front_02_r [{shapeType=Capsule, length=9.7682, radius=3.649, shapeName=dyn_skirt_front_02_r_capsule, center=<-6.4127 0 -2.5>, rotation=<rot 0 0 90>}]
 dyn_skirt_front_03_l [{shapeType=Capsule, length=6.7827, radius=3.2173, shapeName=dyn_skirt_front_03_l_capsule, center=<6.8291 0 2.7084>, rotation=<rot 90 0 0>}]
 dyn_skirt_front_03_r [{shapeType=Capsule, length=9.8819, radius=4.6353, shapeName=dyn_skirt_front_03_r_capsule, center=<-6.8291 0 -3>, rotation=<rot 0 0 90>}]
 dyn_skirt_front_04_l [{shapeType=Capsule, length=6.5327, radius=2.4673, shapeName=dyn_skirt_front_04_l_capsule, center=<5.1575 0 2.7084>, rotation=<rot 90 0 0>}]
 dyn_skirt_front_04_r [
  {shapeType=Capsule, length=9.4588, radius=4.3084, shapeName=dyn_skirt_front_04_r_capsule, center=<-5.5 0 -2.5638>, rotation=<rot 0 0 90>}
  {
   shapeType=Capsule
   length=18.2117
   radius=2.9881
   shapeName=dyn_skirt_front_04_r_capsule
   center=<-18.6496 5.2339 5.3835>
   rotation=<rot 0 -87.0516 109.5998>
  }
 ]
 dyn_skirt_sideA_01_l [{shapeType=Capsule, length=7.2988, radius=1.3277, shapeName=dyn_skirt_sideA_01_l_capsule, center=<6.8291 0 0.641>, rotation=<rot 90 0 0>}]
 dyn_skirt_sideA_01_l_MSTR [{shapeType=Capsule, length=0.5, radius=0.5, shapeName=dyn_skirt_sideA_01_l_MSTR_capsule}]
 dyn_skirt_sideA_01_r [{shapeType=Capsule, length=12.9369, radius=2.2146, shapeName=dyn_skirt_sideA_01_r_capsule, center=<-9.7213 0 -1.7306>, rotation=<rot 0 0 90>}]
 dyn_skirt_sideA_01_r_MSTR [{shapeType=Capsule, length=0.5, radius=0.5, shapeName=dyn_skirt_sideA_01_r_MSTR_capsule, rotation=<rot 90 0 0>}]
 dyn_skirt_sideA_02_l [{shapeType=Capsule, length=10.4691, radius=3.2809, shapeName=dyn_skirt_sideA_02_l_capsule, center=<6.8291 0 1.0115>, rotation=<rot 90 0 0>}]
 dyn_skirt_sideA_02_r [{shapeType=Capsule, length=10.7961, radius=2.9539, shapeName=dyn_skirt_sideA_02_r_capsule, center=<-9.3585 0 0>, rotation=<rot 90 0 0>}]
 dyn_skirt_sideA_03_l [{shapeType=Capsule, length=5.9691, radius=4.0309, shapeName=dyn_skirt_sideA_03_l_capsule, center=<6.8291 0 1.0115>, rotation=<rot 90 0 0>}]
 dyn_skirt_sideA_03_r [{shapeType=Capsule, length=10.0862, radius=3.2809, shapeName=dyn_skirt_sideA_03_r_capsule, center=<-10.1499 0 0>, rotation=<rot 90 0 0>}]
 dyn_skirt_sideA_04_l [{
   shapeType=Capsule
   length=11.0412
   radius=4.7088
   shapeName=dyn_skirt_sideA_04_l_capsule
   center=<8.9353 -1.6802 0.3231>
   rotation=<rot -14.0233 -180 -101.2114>
  }]
 dyn_skirt_sideA_04_r [{shapeType=Capsule, length=12.3143, radius=2.8833, shapeName=dyn_skirt_sideA_04_r_capsule, center=<-9.2282 0 0>, rotation=<rot 90 0 0>}]
 dyn_skirt_sideB_01_l [{shapeType=Capsule, length=10.3688, radius=3.3812, shapeName=dyn_skirt_sideB_01_l_capsule, center=<6.8291 0 3>, rotation=<rot 90 0 0>}]
 dyn_skirt_sideB_01_l_MSTR [{shapeType=Capsule, length=0.5, radius=0.5, shapeName=dyn_skirt_sideB_01_l_MSTR_capsule}]
 dyn_skirt_sideB_01_r [{shapeType=Capsule, length=10.8894, radius=2.8606, shapeName=dyn_skirt_sideB_01_r_capsule, center=<-6.8291 0 -3>, rotation=<rot 0 0 90>}]
 dyn_skirt_sideB_01_r_MSTR [{shapeType=Capsule, length=0.5, radius=0.5, shapeName=dyn_skirt_sideB_01_r_MSTR_capsule, rotation=<rot 90 0 0>}]
 dyn_skirt_sideB_02_l [{shapeType=Capsule, length=11.3515, radius=7.1627, shapeName=dyn_skirt_sideB_02_l_capsule, center=<6.8291 0 3>, rotation=<rot 0 0 90>}]
 dyn_skirt_sideB_02_r [{shapeType=Capsule, length=10.6918, radius=3.2587, shapeName=dyn_skirt_sideB_02_r_capsule, center=<-6.8291 0 -3>, rotation=<rot 0 0 90>}]
 dyn_skirt_sideB_03_l [{shapeType=Capsule, length=24.0574, radius=3.6668, shapeName=dyn_skirt_sideB_03_l_capsule, center=<6.8291 0 0>, rotation=<rot 0 0 90>}]
 dyn_skirt_sideB_03_r [{shapeType=Capsule, length=10.6783, radius=4.1279, shapeName=dyn_skirt_sideB_03_r_capsule, center=<-10.1537 0 -3>, rotation=<rot 0 0 90>}]
 dyn_skirt_sideB_04_l [{shapeType=Capsule, length=15.6407, radius=5.6647, shapeName=dyn_skirt_sideB_04_l_capsule, center=<12.7759 0 0>, rotation=<rot 90 0 0>}]
 pelvis [
  {shapeType=Capsule, length=8, radius=7.75, shapeName=pelvis_capsule}
  {shapeType=Capsule, length=5.25, radius=8.75, shapeName=pelvis_capsule, center=<0 -2.2098 0>}
 ]
 root [
  {shapeType=Capsule, length=0, radius=27.8343, shapeName=root_sphere, center=<0.0002 2.4787 8.5972>, rotation=<rot 0 -90 0>}
  {shapeType=Capsule, length=0, radius=183.4375, shapeName=root_sphere, center=<0 -8.6403 -175.5073>, rotation=<rot 0 -90 0>}
 ]
 spine_01 [{shapeType=Capsule, length=3, radius=8.25, shapeName=spine_01_capsule}]
 thigh_l [
  {shapeType=Capsule, length=34.8334, radius=8.4166, shapeName=thigh_l_capsule, center=<-19 0 1.25>, rotation=<rot 85 179.9996 -177.5091>}
  {shapeType=Capsule, length=0, radius=8, shapeName=thigh_l_capsule, center=<-3.1 -2.7 3.1>, rotation=<rot 90 0 0>}
  {
   shapeType=Capsule
   length=25
   radius=5
   shapeName=thigh_l_capsule
   center=<-24.6187 12.4003 4.4618>
   rotation=<rot 81.7406 179.9753 -179.8298>
  }
 ]
 thigh_r [
  {shapeType=Capsule, length=29.7, radius=9.5, shapeName=thigh_r_capsule, center=<19.8 1.6 -1.1>, rotation=<rot 85 179 179>}
  {shapeType=Capsule, length=0, radius=8, shapeName=thigh_r_capsule, center=<3.1 2.7 -3.1>}
  {
   shapeType=Capsule
   length=20.9298
   radius=3.9702
   shapeName=thigh_r_capsule
   center=<22.9258 -12.7869 -4.9409>
   rotation=<rot -80.299 -0.0001 0>
  }
 ]
}
constraints [
 {
  bone1Name=spine_01
  bone2Name=pelvis
  swing1LimitDegrees=45
  swing1Motion=Free
  swing2LimitDegrees=45
  swing2Motion=Free
  twistLimitDegrees=45
  twistMotion=Free
 }
 {
  bone1Name=dyn_skirt_sideA_01_l_MSTR
  bone2Name=spine_01
  swing1LimitDegrees=45
  swing1Motion=Locked
  swing2LimitDegrees=45
  swing2Motion=Locked
  twistLimitDegrees=45
  twistMotion=Locked
 }
 {bone1Name=dyn_skirt_sideA_01_l, bone2Name=dyn_skirt_sideA_01_l_MSTR, swing1LimitDegrees=45, swing2LimitDegrees=45, twistLimitDegrees=45}
 {bone1Name=dyn_skirt_sideA_02_l, bone2Name=dyn_skirt_sideA_01_l, swing1LimitDegrees=45, swing2LimitDegrees=45, twistLimitDegrees=45}
 {bone1Name=dyn_skirt_sideA_03_l, bone2Name=dyn_skirt_sideA_02_l, swing1LimitDegrees=45, swing2LimitDegrees=45, twistLimitDegrees=5}
 {bone1Name=dyn_skirt_sideA_04_l, bone2Name=dyn_skirt_sideA_03_l, swing1LimitDegrees=45, swing2LimitDegrees=45, twistLimitDegrees=5}
 {
  bone1Name=dyn_skirt_back_01_c_MSTR
  bone2Name=spine_01
  swing1LimitDegrees=45
  swing1Motion=Locked
  swing2LimitDegrees=45
  swing2Motion=Locked
  twistLimitDegrees=45
  twistMotion=Locked
 }
 {bone1Name=dyn_skirt_back_01_c, bone2Name=dyn_skirt_back_01_c_MSTR, swing1LimitDegrees=45, swing2LimitDegrees=45, twistLimitDegrees=45}
 {bone1Name=dyn_skirt_back_02_c, bone2Name=dyn_skirt_back_01_c, swing1LimitDegrees=45, swing2LimitDegrees=45, twistLimitDegrees=45}
 {bone1Name=dyn_skirt_back_03_c, bone2Name=dyn_skirt_back_02_c, swing1LimitDegrees=45, swing2LimitDegrees=45, twistLimitDegrees=45}
 {
  bone1Name=dyn_skirt_sideB_01_l_MSTR
  bone2Name=spine_01
  swing1LimitDegrees=45
  swing1Motion=Locked
  swing2LimitDegrees=45
  swing2Motion=Locked
  twistLimitDegrees=45
  twistMotion=Locked
 }
 {bone1Name=dyn_skirt_sideB_01_l, bone2Name=dyn_skirt_sideB_01_l_MSTR, swing1LimitDegrees=45, swing2LimitDegrees=45, twistLimitDegrees=45}
 {bone1Name=dyn_skirt_sideB_02_l, bone2Name=dyn_skirt_sideB_01_l, swing1LimitDegrees=45, swing2LimitDegrees=45, twistLimitDegrees=45}
 {bone1Name=dyn_skirt_sideB_03_l, bone2Name=dyn_skirt_sideB_02_l, swing1LimitDegrees=45, swing2LimitDegrees=45, twistLimitDegrees=45}
 {bone1Name=dyn_skirt_sideB_04_l, bone2Name=dyn_skirt_sideB_03_l, swing1LimitDegrees=45, swing2LimitDegrees=45, twistLimitDegrees=45}
 {
  bone1Name=dyn_skirt_front_01_l_MSTR
  bone2Name=spine_01
  swing1LimitDegrees=45
  swing1Motion=Locked
  swing2LimitDegrees=45
  swing2Motion=Locked
  twistLimitDegrees=45
  twistMotion=Locked
 }
 {bone1Name=dyn_skirt_front_01_l, bone2Name=dyn_skirt_front_01_l_MSTR, swing1LimitDegrees=45, swing2LimitDegrees=45, twistLimitDegrees=45}
 {bone1Name=dyn_skirt_front_02_l, bone2Name=dyn_skirt_front_01_l, swing1LimitDegrees=45, swing2LimitDegrees=45, twistLimitDegrees=45}
 {bone1Name=dyn_skirt_front_03_l, bone2Name=dyn_skirt_front_02_l, swing1LimitDegrees=45, swing2LimitDegrees=45, twistLimitDegrees=45}
 {bone1Name=dyn_skirt_front_04_l, bone2Name=dyn_skirt_front_03_l, swing1LimitDegrees=45, swing2LimitDegrees=45, twistLimitDegrees=45}
 {
  bone1Name=dyn_skirt_back_01_l_MSTR
  bone2Name=spine_01
  swing1LimitDegrees=45
  swing1Motion=Locked
  swing2LimitDegrees=45
  swing2Motion=Locked
  twistLimitDegrees=45
  twistMotion=Locked
 }
 {bone1Name=dyn_skirt_back_01_l, bone2Name=dyn_skirt_back_01_l_MSTR, swing1LimitDegrees=10, swing2LimitDegrees=45, twistLimitDegrees=5}
 {bone1Name=dyn_skirt_back_02_l, bone2Name=dyn_skirt_back_01_l, swing1LimitDegrees=20, swing2LimitDegrees=45, twistLimitDegrees=15}
 {bone1Name=dyn_skirt_back_03_l, bone2Name=dyn_skirt_back_02_l, swing1LimitDegrees=45, swing2LimitDegrees=45, twistLimitDegrees=45}
 {bone1Name=dyn_skirt_back_04_l, bone2Name=dyn_skirt_back_03_l, swing1LimitDegrees=45, swing2LimitDegrees=45, twistLimitDegrees=45}
 {
  bone1Name=dyn_skirt_sideA_01_r_MSTR
  bone2Name=spine_01
  swing1LimitDegrees=45
  swing1Motion=Locked
  swing2LimitDegrees=45
  swing2Motion=Locked
  twistLimitDegrees=45
  twistMotion=Locked
 }
 {bone1Name=dyn_skirt_sideA_01_r, bone2Name=dyn_skirt_sideA_01_r_MSTR, swing1LimitDegrees=45, swing2LimitDegrees=45, twistLimitDegrees=45}
 {bone1Name=dyn_skirt_sideA_02_r, bone2Name=dyn_skirt_sideA_01_r, swing1LimitDegrees=45, swing2LimitDegrees=45, twistLimitDegrees=45}
 {bone1Name=dyn_skirt_sideA_03_r, bone2Name=dyn_skirt_sideA_02_r, swing1LimitDegrees=45, swing2LimitDegrees=45, twistLimitDegrees=45}
 {bone1Name=dyn_skirt_sideA_04_r, bone2Name=dyn_skirt_sideA_03_r, swing1LimitDegrees=45, swing2LimitDegrees=45, twistLimitDegrees=45}
 {
  bone1Name=dyn_skirt_sideB_01_r_MSTR
  bone2Name=spine_01
  swing1LimitDegrees=45
  swing1Motion=Locked
  swing2LimitDegrees=45
  swing2Motion=Locked
  twistLimitDegrees=45
  twistMotion=Locked
 }
 {bone1Name=dyn_skirt_sideB_01_r, bone2Name=dyn_skirt_sideB_01_r_MSTR, swing1LimitDegrees=20, swing2LimitDegrees=45, twistLimitDegrees=15}
 {bone1Name=dyn_skirt_sideB_02_r, bone2Name=dyn_skirt_sideB_01_r, swing1LimitDegrees=20, swing2LimitDegrees=45, twistLimitDegrees=15}
 {bone1Name=dyn_skirt_sideB_03_r, bone2Name=dyn_skirt_sideB_02_r, swing1LimitDegrees=45, swing2LimitDegrees=45, twistLimitDegrees=45}
 {
  bone1Name=dyn_skirt_back_01_r_MSTR
  bone2Name=spine_01
  swing1LimitDegrees=45
  swing1Motion=Locked
  swing2LimitDegrees=45
  swing2Motion=Locked
  twistLimitDegrees=45
  twistMotion=Locked
 }
 {bone1Name=dyn_skirt_back_01_r, bone2Name=dyn_skirt_back_01_r_MSTR, swing1LimitDegrees=10, swing2LimitDegrees=45, twistLimitDegrees=5}
 {bone1Name=dyn_skirt_back_02_r, bone2Name=dyn_skirt_back_01_r, swing1LimitDegrees=45, swing2LimitDegrees=45, twistLimitDegrees=45}
 {bone1Name=dyn_skirt_back_03_r, bone2Name=dyn_skirt_back_02_r, swing1LimitDegrees=45, swing2LimitDegrees=45, twistLimitDegrees=45}
 {bone1Name=dyn_skirt_back_04_r, bone2Name=dyn_skirt_back_03_r, swing1LimitDegrees=45, swing2LimitDegrees=45, twistLimitDegrees=45}
 {
  bone1Name=dyn_skirt_front_01_r_MSTR
  bone2Name=spine_01
  swing1LimitDegrees=45
  swing1Motion=Locked
  swing2LimitDegrees=45
  swing2Motion=Locked
  twistLimitDegrees=45
  twistMotion=Locked
 }
 {bone1Name=dyn_skirt_front_01_r, bone2Name=dyn_skirt_front_01_r_MSTR, swing1LimitDegrees=45, swing2LimitDegrees=45, twistLimitDegrees=45}
 {bone1Name=dyn_skirt_front_02_r, bone2Name=dyn_skirt_front_01_r, swing1LimitDegrees=45, swing2LimitDegrees=45, twistLimitDegrees=45}
 {bone1Name=dyn_skirt_front_03_r, bone2Name=dyn_skirt_front_02_r, swing1LimitDegrees=45, swing2LimitDegrees=45, twistLimitDegrees=45}
 {bone1Name=dyn_skirt_front_04_r, bone2Name=dyn_skirt_front_03_r, swing1LimitDegrees=45, swing2LimitDegrees=45, twistLimitDegrees=45}
 {
  bone1Name=thigh_l
  bone2Name=pelvis
  swing1LimitDegrees=45
  swing1Motion=Free
  swing2LimitDegrees=45
  swing2Motion=Free
  twistLimitDegrees=45
  twistMotion=Free
 }
 {
  bone1Name=calf_l
  bone2Name=thigh_l
  swing1LimitDegrees=45
  swing1Motion=Free
  swing2LimitDegrees=45
  swing2Motion=Free
  twistLimitDegrees=45
  twistMotion=Free
 }
 {
  bone1Name=thigh_r
  bone2Name=pelvis
  swing1LimitDegrees=45
  swing1Motion=Free
  swing2LimitDegrees=45
  swing2Motion=Free
  twistLimitDegrees=45
  twistMotion=Free
 }
 {
  bone1Name=calf_r
  bone2Name=thigh_r
  swing1LimitDegrees=45
  swing1Motion=Free
  swing2LimitDegrees=45
  swing2Motion=Free
  twistLimitDegrees=45
  twistMotion=Free
 }
 {
  bone1Name=pelvis
  bone2Name=root
  swing1LimitDegrees=45
  swing1Motion=Free
  swing2LimitDegrees=45
  swing2Motion=Free
  twistLimitDegrees=45
  twistMotion=Free
 }
 {
  bone1Name=dyn_skirt_front_02_r
  bone2Name=dyn_skirt_sideA_02_r
  swing1LimitDegrees=45
  swing1Motion=Free
  swing2LimitDegrees=45
  swing2Motion=Free
  twistLimitDegrees=45
  twistMotion=Free
 }
 {
  bone1Name=dyn_skirt_front_03_r
  bone2Name=dyn_skirt_sideA_03_r
  swing1LimitDegrees=45
  swing1Motion=Free
  swing2LimitDegrees=45
  swing2Motion=Free
  twistLimitDegrees=45
  twistMotion=Free
 }
 {
  bone1Name=dyn_skirt_front_03_l
  bone2Name=dyn_skirt_sideA_02_l
  swing1LimitDegrees=45
  swing1Motion=Free
  swing2LimitDegrees=45
  swing2Motion=Free
  twistLimitDegrees=45
  twistMotion=Free
 }
]
