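# Animation Blueprint for the female Vampire Hunter body cosmetic, copying the parent mesh pose and adding deformation, collar and skirt motion.
# Two Control Rigs and collar constraints feed cached poses, with skirt correctives driven by thigh rotation and movement.
# Separate RigidBody simulations for standing, crouching and skydiving are blended by state, then constraints tidy the skirt, including while airborne.
# EventGraph's final crouch-flag write depends only on crouch-sprint and movement, overriding bIsCrouching.
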AnimBlueprint /BRCosmetics/Characters/Player/Female/Medium/Bodies/F_MED_Vampire_Hunter/Meshes/F_MED_Vampire_Hunter_AnimBP
target_skeleton=&/F_MED_Vampire_Hunter_Skeleton
graphs [
 AnimGraph {nodes [
   n1 = RigidBody!(
    ComponentPose=Local To Component(LocalPose=Use cached pose 'CacheMeoutside'())
    bAlphaBoolEnabled:bool
    OverrideWorldGravity:Vector="0,0,-980"
    ExternalForce
   ) -> (Pose) {
    alphaInputType=Bool
    bClampLinearTranslationLimitToRefPose=true
    bOverrideWorldGravity=true
    overridePhysicsAsset=&/F_MED_Vampire_Hunter_Standing_Physics
    simulationSpace=BaseBoneSpace
    alphaBoolBlend {blendOption=Linear}
    baseBoneRef {boneName=pelvis}
    componentAppliedLinearAccClamp=<2500 2500 2500>
    componentLinearAccScale=<0.32 0.32 0.32>
    overrideWorldGravity {z=-980}
    simSpaceSettings {
     maxAngularAcceleration=900
     maxAngularVelocity=2
     maxLinearAcceleration=500
     maxLinearVelocity=500
     velocityScaleZ=0
     worldAlpha=0.2
     externalLinearDragV=<0.2 0.2>
    }
   }
   n2 = "Transform (Modify) Bone - Bone: dyn_skirt_front_01_r_MSTR"(ComponentPose=@n1, Alpha=1, Translation="-0.002,1.146,-1.050", Rotation="0,0,0", Scale="1,1,1") {
    rotationMode=BMM_Additive
    rotationSpace=BCS_BoneSpace
    translationMode=BMM_Additive
    translationSpace=BCS_BoneSpace
    alphaBoolBlend {blendOption=Linear}
    boneToModify {boneName=dyn_skirt_front_01_r_MSTR}
    translation=<-0.002 1.146 -1.05>
   }
   n3 = "Transform (Modify) Bone - Bone: dyn_skirt_back_01_l_MSTR"!(
    ComponentPose=Local To Component(LocalPose=Use cached pose 'CacheMeoutside'())
    bAlphaBoolEnabled:bool=Get new Is crouching() -> (new Is crouching:bool)
    Translation="0,0,0"
    Rotation="30,0,0"
    Scale="1,1,1"
   ) -> (Pose) {
    alphaInputType=Bool
    rotationMode=BMM_Additive
    rotationSpace=BCS_BoneSpace
    translationMode=BMM_Additive
    translationSpace=BCS_BoneSpace
    alphaBoolBlend {blendOption=Linear}
    alphaScaleBiasClamp {bInterpResult=true, bMapRange=true}
    boneToModify {boneName=dyn_skirt_back_01_l_MSTR}
    rotation {pitch=30}
   }
   n4 = "Transform (Modify) Bone - Bone: dyn_skirt_back_01_r_MSTR"!(ComponentPose=@n3, bAlphaBoolEnabled:bool=Get new Is crouching(), Translation="0,0,0", Rotation="30,0,0", Scale="1,1,1") -> (Pose) {
    alphaInputType=Bool
    rotationMode=BMM_Additive
    rotationSpace=BCS_BoneSpace
    translationMode=BMM_Additive
    translationSpace=BCS_BoneSpace
    alphaBoolBlend {blendOption=Linear}
    alphaScaleBiasClamp {bInterpResult=true, bMapRange=true}
    boneToModify {boneName=dyn_skirt_back_01_r_MSTR}
    rotation {pitch=30}
   }
   n5 = RigidBody(ComponentPose=@n4, OverrideWorldGravity="0,0,-980") {
    alphaInputType=Bool
    bClampLinearTranslationLimitToRefPose=true
    bOverrideWorldGravity=true
    overridePhysicsAsset=&/F_MED_Vampire_Hunter_Crouching_Physics
    simulationSpace=BaseBoneSpace
    alphaBoolBlend {blendOption=Linear}
    baseBoneRef {boneName=pelvis}
    componentAppliedLinearAccClamp=<2000 2000 2000>
    componentLinearAccScale=<0.4 0.4 0.4>
    overrideWorldGravity {z=-980}
    simSpaceSettings {
     maxAngularAcceleration=500
     maxAngularVelocity=3
     maxLinearAcceleration=250
     maxLinearVelocity=250
     velocityScaleZ=0
     worldAlpha=0.1
     externalLinearDragV=<0.1 0.1>
    }
   }
   n6 = Blend Poses by bool(
    BlendPose_0=Component To Local(ComponentPose=@n5)
    BlendPose_1=Component To Local(ComponentPose=@n2)
    BlendTime_0=0.5
    BlendTime_1=0.5
    bActiveValue=Get new Is crouching()
   ) {blendType=Linear}
   n7 = "Transform (Modify) Bone - Bone: dyn_skirt_front_01_l_MSTR"(
    ComponentPose=Local To Component(LocalPose=Use cached pose 'AnimCache'())
    Alpha=1
    Translation="0,0,0"
    Rotation="16.357,-9.389,-0.738"
    Scale="1,1,1"
   ) {
    rotationMode=BMM_Additive
    rotationSpace=BCS_BoneSpace
    translationMode=BMM_Additive
    translationSpace=BCS_BoneSpace
    alphaBoolBlend {blendOption=Linear}
    alphaScaleBiasClamp {bInterpResult=true, bMapRange=true}
    boneToModify {boneName=dyn_skirt_front_01_l_MSTR}
    rotation=<rot 16.357 -9.389 -0.738>
   }
   n8 = "Transform (Modify) Bone - Bone: dyn_skirt_front_02_l"(ComponentPose=@n7, Alpha=1, Translation="0,0,0", Rotation="3.531,-5.710,-1.245", Scale="1,1,1") {
    rotationMode=BMM_Additive
    rotationSpace=BCS_BoneSpace
    alphaBoolBlend {blendOption=Linear}
    alphaScaleBiasClamp {bInterpResult=true}
    boneToModify {boneName=dyn_skirt_front_02_l}
    rotation=<rot 3.531 -5.71 -1.245>
   }
   n9 = "Transform (Modify) Bone - Bone: dyn_skirt_front_03_l"(ComponentPose=@n8, Alpha=1, Translation="0,0,0", Rotation="-8.300,-0.177,0.011", Scale="1,1,1") {
    rotationMode=BMM_Additive
    rotationSpace=BCS_BoneSpace
    alphaBoolBlend {blendOption=Linear}
    alphaScaleBiasClamp {bInterpResult=true}
    boneToModify {boneName=dyn_skirt_front_03_l}
    rotation=<rot -8.3 -0.177 0.011>
   }
   n10 = "Transform (Modify) Bone - Bone: dyn_skirt_front_04_l"(ComponentPose=@n9, Alpha=1, Translation="0,0,0", Rotation="9.192,11.395,1.854", Scale="1,1,1") {
    rotationMode=BMM_Additive
    rotationSpace=BCS_BoneSpace
    alphaBoolBlend {blendOption=Linear}
    alphaScaleBiasClamp {bInterpResult=true}
    boneToModify {boneName=dyn_skirt_front_04_l}
    rotation=<rot 9.192 11.395 1.854>
   }
   n11 = "Transform (Modify) Bone - Bone: dyn_skirt_sideA_01_l_MSTR"(ComponentPose=@n10, Alpha=1, Translation="0,0,0", Rotation="14.570,-2.616,0.104", Scale="1,1,1") {
    rotationMode=BMM_Additive
    rotationSpace=BCS_BoneSpace
    alphaBoolBlend {blendOption=Linear}
    alphaScaleBiasClamp {bInterpResult=true}
    boneToModify {boneName=dyn_skirt_sideA_01_l_MSTR}
    rotation=<rot 14.57 -2.616 0.104>
   }
   n12 = "Transform (Modify) Bone - Bone: dyn_skirt_sideA_02_l"(ComponentPose=@n11, Alpha=1, Translation="0,0,0", Rotation="3.821,0.027,-0.683", Scale="1,1,1") {
    rotationMode=BMM_Additive
    rotationSpace=BCS_BoneSpace
    alphaBoolBlend {blendOption=Linear}
    alphaScaleBiasClamp {bInterpResult=true}
    boneToModify {boneName=dyn_skirt_sideA_02_l}
    rotation=<rot 3.821 0.027 -0.683>
   }
   n13 = "Transform (Modify) Bone - Bone: dyn_skirt_sideA_03_l"(ComponentPose=@n12, Alpha=1, Translation="0,0,0", Rotation="0,0,0", Scale="1,1,1") {
    rotationMode=BMM_Additive
    rotationSpace=BCS_BoneSpace
    alphaBoolBlend {blendOption=Linear}
    alphaScaleBiasClamp {bInterpResult=true}
    boneToModify {boneName=dyn_skirt_sideA_03_l}
   }
   n14 = "Transform (Modify) Bone - Bone: dyn_skirt_sideA_04_l"(ComponentPose=@n13, Alpha=1, Translation="0,0,0", Rotation="0,0,0", Scale="1,1,1") {
    rotationMode=BMM_Additive
    rotationSpace=BCS_BoneSpace
    alphaBoolBlend {blendOption=Linear}
    alphaScaleBiasClamp {bInterpResult=true}
    boneToModify {boneName=dyn_skirt_sideA_04_l}
   }
   n15 = "Transform (Modify) Bone - Bone: dyn_skirt_sideB_01_l_MSTR"(ComponentPose=@n14, Alpha=1, Translation="0,0,0", Rotation="9.517,-10.083,-0.086", Scale="1,1,1") {
    rotationMode=BMM_Additive
    rotationSpace=BCS_BoneSpace
    alphaBoolBlend {blendOption=Linear}
    alphaScaleBiasClamp {bInterpResult=true}
    boneToModify {boneName=dyn_skirt_sideB_01_l_MSTR}
    rotation=<rot 9.517 -10.083 -0.086>
   }
   n16 = "Transform (Modify) Bone - Bone: dyn_skirt_sideB_02_l"(ComponentPose=@n15, Alpha=1, Translation="0,0,0", Rotation="0.016,9.608,0.006", Scale="1,1,1") {
    rotationMode=BMM_Additive
    rotationSpace=BCS_BoneSpace
    alphaBoolBlend {blendOption=Linear}
    alphaScaleBiasClamp {bInterpResult=true}
    boneToModify {boneName=dyn_skirt_sideB_02_l}
    rotation=<rot 0.016 9.608 0.006>
   }
   n17 = "Transform (Modify) Bone - Bone: dyn_skirt_front_01_r_MSTR"(ComponentPose=@n16, Alpha=1, Translation="0,0,0", Rotation="15.214,-6.327,-17.682", Scale="1,1,1") {
    rotationMode=BMM_Additive
    rotationSpace=BCS_BoneSpace
    translationMode=BMM_Additive
    translationSpace=BCS_BoneSpace
    alphaBoolBlend {blendOption=Linear}
    alphaScaleBiasClamp {bInterpResult=true, bMapRange=true}
    boneToModify {boneName=dyn_skirt_front_01_r_MSTR}
    rotation=<rot 15.214 -6.327 -17.682>
   }
   n18 = "Transform (Modify) Bone - Bone: dyn_skirt_front_02_r"(ComponentPose=@n17, Alpha=1, Translation="0,0,0", Rotation="-7.721,-2.537,-0.005", Scale="1,1,1") {
    rotationMode=BMM_Additive
    rotationSpace=BCS_BoneSpace
    alphaBoolBlend {blendOption=Linear}
    alphaScaleBiasClamp {bInterpResult=true}
    boneToModify {boneName=dyn_skirt_front_02_r}
    rotation=<rot -7.721 -2.537 -0.005>
   }
   n19 = "Transform (Modify) Bone - Bone: dyn_skirt_front_03_r"(ComponentPose=@n18, Alpha=1, Translation="0,0,0", Rotation="0,0,0", Scale="1,1,1") {
    rotationMode=BMM_Additive
    rotationSpace=BCS_BoneSpace
    alphaBoolBlend {blendOption=Linear}
    alphaScaleBiasClamp {bInterpResult=true}
    boneToModify {boneName=dyn_skirt_front_03_r}
   }
   n20 = "Transform (Modify) Bone - Bone: dyn_skirt_sideA_01_r_MSTR"(ComponentPose=@n19, Alpha=1, Translation="0,0,0", Rotation="12.652,5.235,-5.406", Scale="1,1,1") {
    rotationMode=BMM_Additive
    rotationSpace=BCS_BoneSpace
    translationMode=BMM_Additive
    translationSpace=BCS_BoneSpace
    alphaBoolBlend {blendOption=Linear}
    alphaScaleBiasClamp {bInterpResult=true, bMapRange=true}
    boneToModify {boneName=dyn_skirt_sideA_01_r_MSTR}
    rotation=<rot 12.652 5.235 -5.406>
   }
   n21 = "Transform (Modify) Bone - Bone: dyn_skirt_sideA_02_r"(ComponentPose=@n20, Alpha=1, Translation="0,0,0", Rotation="0,0,0", Scale="1,1,1") {
    rotationMode=BMM_Additive
    rotationSpace=BCS_BoneSpace
    alphaBoolBlend {blendOption=Linear}
    alphaScaleBiasClamp {bInterpResult=true}
    boneToModify {boneName=dyn_skirt_sideA_02_r}
   }
   n22 = "Transform (Modify) Bone - Bone: dyn_skirt_sideA_03_r"(ComponentPose=@n21, Alpha=1, Translation="0,0,0", Rotation="0,0,0", Scale="1,1,1") {
    rotationMode=BMM_Additive
    rotationSpace=BCS_BoneSpace
    alphaBoolBlend {blendOption=Linear}
    alphaScaleBiasClamp {bInterpResult=true}
    boneToModify {boneName=dyn_skirt_sideA_03_r}
   }
   n23 = "Transform (Modify) Bone - Bone: dyn_skirt_sideA_04_r"(ComponentPose=@n22, Alpha=1, Translation="0,0,0", Rotation="0,0,0", Scale="1,1,1") {
    rotationMode=BMM_Additive
    rotationSpace=BCS_BoneSpace
    alphaBoolBlend {blendOption=Linear}
    alphaScaleBiasClamp {bInterpResult=true}
    boneToModify {boneName=dyn_skirt_sideA_04_r}
   }
   n24 = "Transform (Modify) Bone - Bone: dyn_skirt_sideB_01_r_MSTR"(ComponentPose=@n23, Alpha=1, Translation="0,0,0", Rotation="13.329,-5.479,-0.136", Scale="1,1,1") {
    rotationMode=BMM_Additive
    rotationSpace=BCS_BoneSpace
    translationMode=BMM_Additive
    translationSpace=BCS_BoneSpace
    alphaBoolBlend {blendOption=Linear}
    alphaScaleBiasClamp {bInterpResult=true, bMapRange=true}
    boneToModify {boneName=dyn_skirt_sideB_01_r_MSTR}
    rotation=<rot 13.329 -5.479 -0.136>
   }
   n25 = "Transform (Modify) Bone - Bone: dyn_skirt_sideB_02_r"(ComponentPose=@n24, Alpha=1, Translation="0,0,0", Rotation="-0.000,3.090,0.008", Scale="1,1,1") {
    rotationMode=BMM_Additive
    rotationSpace=BCS_BoneSpace
    alphaBoolBlend {blendOption=Linear}
    alphaScaleBiasClamp {bInterpResult=true}
    boneToModify {boneName=dyn_skirt_sideB_02_r}
    rotation=<rot 0 3.0901 0.008>
   }
   n26 = "Transform (Modify) Bone - Bone: dyn_skirt_sideB_03_r"(ComponentPose=@n25, Alpha=1, Translation="0,0,0", Rotation="0,0,0", Scale="0.9,1,1") {
    rotationMode=BMM_Additive
    rotationSpace=BCS_BoneSpace
    scaleMode=BMM_Replace
    alphaBoolBlend {blendOption=Linear}
    alphaScaleBiasClamp {bInterpResult=true}
    boneToModify {boneName=dyn_skirt_sideB_03_r}
    scale {x=0.9}
   }
   n27 = "Transform (Modify) Bone - Bone: dyn_skirt_sideB_04_r"(ComponentPose=@n26, Alpha=1, Translation="0,0,0", Rotation="0,0,0", Scale="1,1,1") {
    rotationMode=BMM_Additive
    rotationSpace=BCS_BoneSpace
    scaleMode=BMM_Replace
    alphaBoolBlend {blendOption=Linear}
    alphaScaleBiasClamp {bInterpResult=true}
    boneToModify {boneName=dyn_skirt_sideB_04_r}
   }
   n28 = RigidBody!(ComponentPose=@n27, Alpha:float, OverrideWorldGravity="0,0,-980", ExternalForce) -> (Pose) {
    bClampLinearTranslationLimitToRefPose=true
    bOverrideWorldGravity=true
    overridePhysicsAsset=&/F_MED_Vampire_Hunter_Skydiving_Physics
    simulationSpace=BaseBoneSpace
    alphaBoolBlend {blendOption=Linear}
    baseBoneRef {boneName=pelvis}
    componentAppliedLinearAccClamp=<2500 2500 2500>
    componentLinearAccScale=<0.4 0.4 0.42>
    overrideWorldGravity {z=-980}
    simSpaceSettings {
     maxAngularAcceleration=1000
     maxAngularVelocity=3
     maxLinearAcceleration=500
     maxLinearVelocity=500
     velocityScaleZ=0
     worldAlpha=0.45
     externalLinearDragV=<0.1 0.1>
    }
   }
   n29 = Get bIsSkydiveDiveMode()
   n30 = "Transform (Modify) Bone - Bone: dyn_skirt_sideA_01_r_MSTR"!(ComponentPose=@n28, bAlphaBoolEnabled:bool=@n29, Translation="0,0,0", Rotation="-10,0,0", Scale="1,1,1") -> (Pose) {
    alphaInputType=Bool
    rotationMode=BMM_Additive
    rotationSpace=BCS_BoneSpace
    scaleMode=BMM_Additive
    scaleSpace=BCS_BoneSpace
    translationMode=BMM_Additive
    translationSpace=BCS_BoneSpace
    alphaBoolBlend {blendInTime=0.1, blendOption=Linear, blendOutTime=0.1}
    boneToModify {boneName=dyn_skirt_sideA_01_r_MSTR}
    rotation {pitch=-10}
   }
   n31 = "Transform (Modify) Bone - Bone: dyn_skirt_sideB_01_r_MSTR"!(ComponentPose=@n30, bAlphaBoolEnabled:bool=@n29, Translation="0,0,0", Rotation="-10,0,0", Scale="1,1,1") -> (Pose) {
    alphaInputType=Bool
    rotationMode=BMM_Additive
    rotationSpace=BCS_BoneSpace
    scaleMode=BMM_Additive
    scaleSpace=BCS_BoneSpace
    translationMode=BMM_Additive
    translationSpace=BCS_BoneSpace
    alphaBoolBlend {blendInTime=0.1, blendOption=Linear, blendOutTime=0.1}
    boneToModify {boneName=dyn_skirt_sideB_01_r_MSTR}
    rotation {pitch=-10}
   }
   n32 = "Transform (Modify) Bone - Bone: dyn_skirt_sideA_01_l_MSTR"!(ComponentPose=@n31, bAlphaBoolEnabled:bool=@n29, Translation="0,0,0", Rotation="-12,0,0", Scale="1,1,1") -> (Pose) {
    alphaInputType=Bool
    rotationMode=BMM_Additive
    rotationSpace=BCS_BoneSpace
    scaleMode=BMM_Additive
    scaleSpace=BCS_BoneSpace
    translationMode=BMM_Additive
    translationSpace=BCS_BoneSpace
    alphaBoolBlend {blendInTime=0.1, blendOption=Linear, blendOutTime=0.1}
    boneToModify {boneName=dyn_skirt_sideA_01_l_MSTR}
    rotation {pitch=-12}
   }
   n33 = "Transform (Modify) Bone - Bone: dyn_skirt_sideB_01_l_MSTR"!(ComponentPose=@n32, bAlphaBoolEnabled:bool=@n29, Translation="0,0,0", Rotation="-12,0,0", Scale="1,1,1") -> (Pose) {
    alphaInputType=Bool
    rotationMode=BMM_Additive
    rotationSpace=BCS_BoneSpace
    scaleMode=BMM_Additive
    scaleSpace=BCS_BoneSpace
    translationMode=BMM_Additive
    translationSpace=BCS_BoneSpace
    alphaBoolBlend {blendInTime=0.1, blendOption=Linear, blendOutTime=0.1}
    boneToModify {boneName=dyn_skirt_sideB_01_l_MSTR}
    rotation {pitch=-12}
   }
   n34 = Component To Local(ComponentPose=@n33)
   n35 = Get bIsSkyDiving()
   n36 = SelectFloat(A=0.3, B=0, bPickA=Get new Is crouching())
   n37 = "Constraint - dyn_skirt_back_02_r (P)"(
    ComponentPose=Local To Component(LocalPose=Blend Poses by bool(BlendPose_0=@n34, BlendPose_1=@n6, bActiveValue=@n35) {blendType=Linear})
    Alpha=@n36
    ConstraintWeights_0=1
   ) {
    alphaBoolBlend {blendOption=Linear}
    boneToModify {boneName=dyn_skirt_back_02_r}
    constraintSetup [{targetBone {boneName=dyn_skirt_back_02_l}}]
    constraintWeights [1]
   }
   n38 = "Constraint - dyn_skirt_back_02_l (P)"(ComponentPose=@n37, Alpha=@n36, ConstraintWeights_0=1) {
    alphaBoolBlend {blendOption=Linear}
    boneToModify {boneName=dyn_skirt_back_02_l}
    constraintSetup [{targetBone {boneName=dyn_skirt_back_02_r}}]
    constraintWeights [1]
   }
   n39 = AND Boolean(A=Get JumpingFalling() -> (JumpingFalling:bool), B=NOT Boolean(A=Get bIsSkyDiving()))
   n40 = "Transform (Modify) Bone - Bone: dyn_skirt_front_01_r_MSTR"!(ComponentPose=@n38, bAlphaBoolEnabled:bool=@n39, Translation="-2,0,2", Rotation="-10,-30,0", Scale="1,1,1") -> (Pose) {
    alphaInputType=Bool
    rotationMode=BMM_Additive
    translationMode=BMM_Additive
    alphaBoolBlend {blendInTime=1, blendOption=Linear, blendOutTime=0.1}
    alphaScaleBiasClamp {bInterpResult=true, bMapRange=true}
    boneToModify {boneName=dyn_skirt_front_01_r_MSTR}
    rotation {pitch=-10, yaw=-30}
    translation {x=-2, z=2}
   }
   n41 = "Constraint - dyn_skirt_sideA_01_r_MSTR (R)"!(ComponentPose=@n40, bAlphaBoolEnabled:bool=@n39, ConstraintWeights_0=1) -> (Pose) {
    alphaInputType=Bool
    alphaBoolBlend {blendInTime=1, blendOption=Linear, blendOutTime=0.5}
    boneToModify {boneName=dyn_skirt_sideA_01_r_MSTR}
    constraintSetup [{transformType=Rotation, targetBone {boneName=dyn_skirt_front_01_r_MSTR}}]
    constraintWeights [1]
   }
   n42 = "Constraint - dyn_skirt_sideA_03_r (R)"!(ComponentPose=@n41, bAlphaBoolEnabled:bool=@n39, ConstraintWeights_0=1) -> (Pose) {
    alphaInputType=Bool
    alphaBoolBlend {blendInTime=1, blendOption=Linear, blendOutTime=0.5}
    boneToModify {boneName=dyn_skirt_sideA_03_r}
    constraintSetup [{transformType=Rotation, targetBone {boneName=dyn_skirt_front_04_r}}]
    constraintWeights [1]
   }
   n43 = "Constraint - dyn_skirt_sideB_01_r_MSTR (R)"!(ComponentPose=@n42, bAlphaBoolEnabled:bool=@n39, ConstraintWeights_0=1) -> (Pose) {
    alphaInputType=Bool
    alphaBoolBlend {blendInTime=1, blendOption=Linear, blendOutTime=0.5}
    boneToModify {boneName=dyn_skirt_sideB_01_r_MSTR}
    constraintSetup [{transformType=Rotation, targetBone {boneName=dyn_skirt_sideA_01_r_MSTR}}]
    constraintWeights [1]
   }
   n44 = "Constraint - dyn_skirt_sideA_03_l (R)"(ComponentPose=@n43, Alpha=0.2, ConstraintWeights_0=1) {
    alphaBoolBlend {blendInTime=1, blendOption=Linear, blendOutTime=0.5}
    boneToModify {boneName=dyn_skirt_sideA_03_l}
    constraintSetup [{transformType=Rotation, targetBone {boneName=dyn_skirt_front_04_l}}]
    constraintWeights [1]
   }
   n45 = "Constraint - dyn_skirt_sideB_03_l (R)"!(ComponentPose=@n44, bAlphaBoolEnabled:bool=@n39, ConstraintWeights_0=0.2) -> (Pose) {
    alphaInputType=Bool
    alphaBoolBlend {blendInTime=1, blendOption=Linear, blendOutTime=0.5}
    boneToModify {boneName=dyn_skirt_sideB_03_l}
    constraintSetup [{transformType=Rotation, targetBone {boneName=dyn_skirt_sideA_03_l}}]
    constraintWeights [1]
   }
   n46 = "Constraint - dyn_skirt_back_01_c (RR)"!(ComponentPose=@n45, Alpha=1, ConstraintWeights_0=0.5, ConstraintWeights_1:float=0.3) -> (Pose) {
    alphaBoolBlend {blendOption=Linear}
    boneToModify {boneName=dyn_skirt_back_01_c}
    constraintSetup [{transformType=Rotation, targetBone {boneName=dyn_skirt_back_01_l}}, {transformType=Rotation, targetBone {boneName=dyn_skirt_back_01_r}}]
    constraintWeights [1,1]
   }
   n47 = "Constraint - dyn_skirt_back_02_c (RR)"!(ComponentPose=@n46, Alpha=1, ConstraintWeights_0=0.5, ConstraintWeights_1:float=0.3) -> (Pose) {
    alphaBoolBlend {blendOption=Linear}
    boneToModify {boneName=dyn_skirt_back_02_c}
    constraintSetup [{transformType=Rotation, targetBone {boneName=dyn_skirt_back_02_l}}, {transformType=Rotation, targetBone {boneName=dyn_skirt_back_02_r}}]
    constraintWeights [1,1]
   }
   Output Pose(Result=Component To Local(ComponentPose=@n47)) {layerGroup=None}
   n48 = Copy Pose From Mesh() {bUseAttachedParent=true}
   n49 = Control Rig(Source=Component To Local(ComponentPose=Local To Component(LocalPose=@n48))) {
    bTransferInputCurves=false
    alphaBoolBlend {blendOption=Linear}
    controlRigAssetReference {blueprintRigClass=&/F_MED_Vampire_HunterDeform_CtrlRig_twists.F_MED_Vampire_HunterDeform_CtrlRig_twists_C}
   }
   n50 = Control Rig!(
    Source=@n49
    Alpha
    KneeRot_fr_l_114deg:Vector
    KneeTrans_fr_l_114deg:Vector="0.258,-3.591,0.009"
    KneeScale_fr_l_114deg:Vector="1,1,1"
    KneeScale_fr_l_130deg:Vector="1,1,1"
    KneeRot_fr_l_130deg:Vector="0,0,-51.2"
    KneeTrans_fr_l_130deg:Vector="0.258,-3.591,0.009"
   ) -> (Pose) {
    bTransferInputCurves=false
    alphaBoolBlend {blendOption=Linear}
    controlRigAssetReference {blueprintRigClass=&/F_MED_Vampire_Hunter_Deform_CtrlRig_sculpters.F_MED_Vampire_Hunter_Deform_CtrlRig_sculpters_C}
   }
   n51 = "Constraint - dyn_collar_l_01 (PPP)"!(
    ComponentPose=Local To Component(LocalPose=@n50)
    Alpha=1
    ConstraintWeights_0=0.5
    ConstraintWeights_1:float=0
    ConstraintWeights_2:float=0.5
   ) -> (Pose) {
    alphaBoolBlend {blendOption=Linear}
    boneToModify {boneName=dyn_collar_l_01}
    constraintSetup [{targetBone {boneName=clavicle_l}}, {targetBone {boneName=upperarm_l}}, {targetBone {boneName=deform_deltoid_fr_l}}]
    constraintWeights [1,1,1]
   }
   n52 = "Constraint - dyn_collar_l_02 (PP)"!(ComponentPose=@n51, Alpha=1, ConstraintWeights_0=0.8, ConstraintWeights_1:float=0.2) -> (Pose) {
    alphaBoolBlend {blendOption=Linear}
    boneToModify {boneName=dyn_collar_l_02}
    constraintSetup [{targetBone {boneName=clavicle_l}}, {targetBone {boneName=upperarm_l}}]
    constraintWeights [1,1]
   }
   n53 = "Constraint - dyn_collar_r_01 (PPP)"!(ComponentPose=@n52, Alpha=1, ConstraintWeights_0=0.5, ConstraintWeights_1:float=0, ConstraintWeights_2:float=0.5) -> (Pose) {
    alphaBoolBlend {blendOption=Linear}
    boneToModify {boneName=dyn_collar_r_01}
    constraintSetup [{targetBone {boneName=clavicle_r}}, {targetBone {boneName=upperarm_r}}, {targetBone {boneName=deform_deltoid_fr_r}}]
    constraintWeights [1,1,1]
   }
   n54 = "Constraint - dyn_collar_r_02 (PP)"!(ComponentPose=@n53, Alpha=1, ConstraintWeights_0=0.8, ConstraintWeights_1:float=0.2) -> (Pose) {
    alphaBoolBlend {blendOption=Linear}
    boneToModify {boneName=dyn_collar_r_02}
    constraintSetup [{targetBone {boneName=clavicle_r}}, {targetBone {boneName=upperarm_r}}]
    constraintWeights [1,1]
   }
   Save cached pose 'AnimCache'(Pose=Component To Local(ComponentPose=@n54)) {cacheName=AnimCache}
   n55 = "Transform (Modify) Bone - Bone: dyn_skirt_front_01_r_MSTR"!(
    ComponentPose=Local To Component(LocalPose=Use cached pose 'AnimCache'())
    Alpha:float=Get thigh_rot_otr_r() -> (thigh_rot_otr_r:double)
    Translation="0,0,-5"
    Rotation="0,0,0"
    Scale="1,1,1"
   ) -> (Pose) {
    rotationMode=BMM_Additive
    rotationSpace=BCS_BoneSpace
    scaleMode=BMM_Additive
    scaleSpace=BCS_BoneSpace
    translationMode=BMM_Additive
    translationSpace=BCS_BoneSpace
    alphaBoolBlend {blendOption=Linear}
    boneToModify {boneName=dyn_skirt_front_01_r_MSTR}
    translation {z=-5}
   }
   n56 = "Transform (Modify) Bone - Bone: dyn_skirt_front_01_r_MSTR"(ComponentPose=@n55, Alpha=Get thigh_rot_up_r() -> (thigh_rot_up_r:double), Translation="0,0,-5", Rotation="30,30,0", Scale="1,1,1") {
    rotationMode=BMM_Additive
    rotationSpace=BCS_BoneSpace
    scaleMode=BMM_Additive
    scaleSpace=BCS_BoneSpace
    translationMode=BMM_Additive
    translationSpace=BCS_BoneSpace
    alphaBoolBlend {blendOption=Linear}
    boneToModify {boneName=dyn_skirt_front_01_r_MSTR}
    rotation {pitch=30, yaw=30}
    translation {z=-5}
   }
   n57 = Get thigh_rot_up_l() -> (thigh_rot_up_l:double)
   n58 = "Transform (Modify) Bone - Bone: dyn_skirt_front_01_l_MSTR"(ComponentPose=@n56, Alpha=@n57, Translation="1.817,0.550,7.513", Rotation="-19.058,0.278,-17.498", Scale="1,1,1") {
    rotationMode=BMM_Additive
    rotationSpace=BCS_BoneSpace
    translationMode=BMM_Additive
    translationSpace=BCS_BoneSpace
    alphaBoolBlend {blendInTime=0.1, blendOption=Linear, blendOutTime=1.5}
    alphaScaleBiasClamp {bInterpResult=true, bMapRange=true}
    boneToModify {boneName=dyn_skirt_front_01_l_MSTR}
    rotation=<rot -19.058 0.278 -17.498>
    translation=<1.817 0.55 7.513>
   }
   n59 = "Transform (Modify) Bone - Bone: dyn_skirt_sideA_01_l_MSTR"!(ComponentPose=@n58, Alpha:float=@n57, Translation="-1.096,0.269,2.487", Rotation="-0.064,-0.021,-20.410", Scale="1,1,1") -> (Pose) {
    rotationMode=BMM_Additive
    rotationSpace=BCS_BoneSpace
    translationMode=BMM_Additive
    translationSpace=BCS_BoneSpace
    alphaBoolBlend {blendInTime=0.1, blendOption=Linear, blendOutTime=1.5}
    alphaScaleBiasClamp {bInterpResult=true, bMapRange=true}
    boneToModify {boneName=dyn_skirt_sideA_01_l_MSTR}
    rotation=<rot -0.064 -0.021 -20.41>
    translation=<-1.096 0.269 2.487>
   }
   n60 = Get thigh_rot_up_l()
   n61 = "Transform (Modify) Bone - Bone: dyn_skirt_front_01_l_MSTR"(ComponentPose=@n59, Alpha=@n60, Translation="-4.596,-5.199,-2.125", Rotation="57.280,-125.155,-164.248", Scale="1,1,1") {
    rotationMode=BMM_Additive
    rotationSpace=BCS_BoneSpace
    translationMode=BMM_Additive
    translationSpace=BCS_BoneSpace
    alphaBoolBlend {blendOption=Linear}
    alphaScaleBiasClamp {bInterpResult=true, bMapRange=true}
    boneToModify {boneName=dyn_skirt_front_01_l_MSTR}
    rotation=<rot 57.28 -125.155 -164.248>
    translation=<-4.596 -5.199 -2.125>
   }
   n62 = "Transform (Modify) Bone - Bone: dyn_skirt_sideA_01_l_MSTR"(ComponentPose=@n61, Alpha=@n60, Translation="-0.108,0.000,3.717", Rotation="0.036,-0.077,-20.190", Scale="1,1,1") {
    rotationMode=BMM_Additive
    rotationSpace=BCS_BoneSpace
    translationMode=BMM_Additive
    translationSpace=BCS_BoneSpace
    alphaBoolBlend {blendOption=Linear}
    alphaScaleBiasClamp {bInterpResult=true, bMapRange=true}
    boneToModify {boneName=dyn_skirt_sideA_01_l_MSTR}
    rotation=<rot 0.036 -0.077 -20.19>
    translation {x=-0.108, z=3.717}
   }
   n63 = "Transform (Modify) Bone - Bone: dyn_skirt_sideA_01_l_MSTR"(ComponentPose=@n62, Alpha=Get PlayerMovement() -> (PlayerMovement:double), Translation="0,0,5", Rotation="0,0,0", Scale="1,1,1") {
    rotationMode=BMM_Additive
    rotationSpace=BCS_BoneSpace
    translationMode=BMM_Additive
    translationSpace=BCS_BoneSpace
    alphaBoolBlend {blendOption=Linear}
    alphaScaleBiasClamp {bInterpResult=true, bMapRange=true}
    boneToModify {boneName=dyn_skirt_sideA_01_l_MSTR}
    translation {z=5}
   }
   n64 = Get thigh_rot_up_r()
   n65 = "Transform (Modify) Bone - Bone: dyn_skirt_front_01_r_MSTR"(ComponentPose=@n63, Alpha=@n64, Translation="0,0,0", Rotation="15.743,-6.088,-1.669", Scale="1,1,1") {
    rotationMode=BMM_Additive
    rotationSpace=BCS_BoneSpace
    translationMode=BMM_Additive
    translationSpace=BCS_BoneSpace
    alphaBoolBlend {blendOption=Linear}
    alphaScaleBiasClamp {bInterpResult=true, bMapRange=true}
    boneToModify {boneName=dyn_skirt_front_01_r_MSTR}
    rotation=<rot 15.743 -6.088 -1.669>
   }
   n66 = "Transform (Modify) Bone - Bone: dyn_skirt_sideA_01_r_MSTR"!(ComponentPose=@n65, Alpha:float=@n64, Translation="0,0,0", Rotation="9.438,-21.262,-5.072", Scale="1,1,1") -> (Pose) {
    rotationMode=BMM_Additive
    rotationSpace=BCS_BoneSpace
    translationMode=BMM_Additive
    translationSpace=BCS_BoneSpace
    alphaBoolBlend {blendOption=Linear}
    alphaScaleBiasClamp {bInterpResult=true, bMapRange=true}
    boneToModify {boneName=dyn_skirt_sideA_01_r_MSTR}
    rotation=<rot 9.4379 -21.2619 -5.072>
   }
   n67 = "Transform (Modify) Bone - Bone: dyn_skirt_sideA_01_r_MSTR"(ComponentPose=@n66, Alpha=1, Translation="0,0,0", Rotation="-0.047,-2.812,0.017", Scale="1,1,1") {
    rotationMode=BMM_Additive
    rotationSpace=BCS_BoneSpace
    translationMode=BMM_Additive
    translationSpace=BCS_BoneSpace
    alphaBoolBlend {blendOption=Linear}
    alphaScaleBiasClamp {bInterpResult=true, bMapRange=true}
    boneToModify {boneName=dyn_skirt_sideA_01_r_MSTR}
    rotation=<rot -0.047 -2.812 0.017>
   }
   n68 = "Transform (Modify) Bone - Bone: dyn_skirt_sideA_02_r"(ComponentPose=@n67, Alpha=1, Translation="0,0,0", Rotation="0,0,0", Scale="1,1,1") {
    rotationMode=BMM_Additive
    rotationSpace=BCS_BoneSpace
    alphaBoolBlend {blendOption=Linear}
    alphaScaleBiasClamp {bInterpResult=true}
    boneToModify {boneName=dyn_skirt_sideA_02_r}
   }
   n69 = "Transform (Modify) Bone - Bone: dyn_skirt_sideA_03_r"(ComponentPose=@n68, Alpha=1, Translation="0,0,0", Rotation="0,0,0", Scale="1,1,1") {
    rotationMode=BMM_Additive
    rotationSpace=BCS_BoneSpace
    alphaBoolBlend {blendOption=Linear}
    alphaScaleBiasClamp {bInterpResult=true}
    boneToModify {boneName=dyn_skirt_sideA_03_r}
   }
   n70 = "Transform (Modify) Bone - Bone: dyn_skirt_sideA_04_r"(ComponentPose=@n69, Alpha=1, Translation="0,0,0", Rotation="0,0,0", Scale="1,1,1") {
    rotationMode=BMM_Additive
    rotationSpace=BCS_BoneSpace
    alphaBoolBlend {blendOption=Linear}
    alphaScaleBiasClamp {bInterpResult=true}
    boneToModify {boneName=dyn_skirt_sideA_04_r}
   }
   n71 = "Transform (Modify) Bone - Bone: dyn_skirt_sideB_01_r_MSTR"!(ComponentPose=@n70, Alpha:float=1, Translation="0,0,0", Rotation="-0.000,-5.344,-0.021", Scale="1,1,1") -> (Pose) {
    rotationMode=BMM_Additive
    rotationSpace=BCS_BoneSpace
    translationMode=BMM_Additive
    translationSpace=BCS_BoneSpace
    alphaBoolBlend {blendOption=Linear}
    alphaScaleBiasClamp {bInterpResult=true, bMapRange=true}
    boneToModify {boneName=dyn_skirt_sideB_01_r_MSTR}
    rotation=<rot 0 -5.344 -0.021>
   }
   n72 = "Transform (Modify) Bone - Bone: dyn_skirt_sideB_02_r"(ComponentPose=@n71, Alpha=1, Translation="0,0,0", Rotation="-0.000,3.090,0.008", Scale="1,1,1") {
    rotationMode=BMM_Additive
    rotationSpace=BCS_BoneSpace
    alphaBoolBlend {blendOption=Linear}
    alphaScaleBiasClamp {bInterpResult=true}
    boneToModify {boneName=dyn_skirt_sideB_02_r}
    rotation=<rot 0 3.0901 0.008>
   }
   n73 = "Transform (Modify) Bone - Bone: dyn_skirt_sideB_03_r"(ComponentPose=@n72, Alpha=1, Translation="0,0,0", Rotation="0,0,0", Scale="0.9,1,1") {
    rotationMode=BMM_Additive
    rotationSpace=BCS_BoneSpace
    scaleMode=BMM_Replace
    alphaBoolBlend {blendOption=Linear}
    alphaScaleBiasClamp {bInterpResult=true}
    boneToModify {boneName=dyn_skirt_sideB_03_r}
    scale {x=0.9}
   }
   n74 = "Transform (Modify) Bone - Bone: dyn_skirt_sideB_04_r"(ComponentPose=@n73, Alpha=1, Translation="0,0,0", Rotation="0,0,0", Scale="1,1,1") {
    rotationMode=BMM_Additive
    rotationSpace=BCS_BoneSpace
    scaleMode=BMM_Replace
    alphaBoolBlend {blendOption=Linear}
    alphaScaleBiasClamp {bInterpResult=true}
    boneToModify {boneName=dyn_skirt_sideB_04_r}
   }
   Save cached pose 'CacheMeoutside'(Pose=Component To Local(ComponentPose=@n74)) {cacheName=CacheMeoutside}
  ]}
 EventGraph {nodes [
   Event BlueprintUpdateAnimation()
   > n1 = Sequence!(execute) -> (then_0, then_1, then_2:Exec, then_3:Exec, then_4:Exec, then_5:Exec, then_6:Exec, then_7:Exec, then_8:Exec, then_9:Exec)
   > Set Body Mesh(Body Mesh:Skeletal Mesh Component Object Reference=GetPartSkeletalMeshComponent()) -> (Output_Get:Skeletal Mesh Component Object Reference)
   Set thigh L Rot(
    execute=@n1.then_1
    thigh L Rot:double=GetAlphaFromReferenceTransformDelta(
     SkeletalMeshComponent=Get Body Mesh() -> (Body Mesh:Skeletal Mesh Component Object Reference)
     BoneName=thigh_l
     BaseName=pelvis
     InputRange=(X=0,Y=-100)
     TwistAxis="1,0,0"
    )
   ) -> (Output_Get:double)
   Set Thigh R Rot(
    execute=@n1.then_2
    Thigh R Rot:double=GetAlphaFromReferenceTransformDelta(SkeletalMeshComponent=Get Body Mesh(), BoneName=thigh_r, BaseName=pelvis, InputRange=(X=0,Y=100), TwistAxis="1,0,0")
   ) -> (Output_Get:double)
   n2 = Branch(execute=@n1.then_3, Condition=Get bIsCrouching())
   > Set new Is crouching(new Is crouching:bool=true) -> (Output_Get:bool)
   Set new Is crouching(execute=@n2.else, new Is crouching=false)
   n3 = Gravity Override State Function DEPRECATED(execute=@n1.then_4, CharacterPartAnimInstance=Self-Reference(), GameTimeInSeconds=0, GravityOverrideParams=Get Gravity Override Params())
   > Set Gravity Override(Gravity Override=@n3)
   n4 = Branch(execute=@n1.then_5, Condition=OR Boolean(A=Get bIsJumping(), B=Get bIsFalling()))
   > Set JumpingFalling(JumpingFalling:bool=true) -> (Output_Get:bool)
   Set JumpingFalling(execute=@n4.else, JumpingFalling=false)
   n5 = PlayerMovement(execute=@n1.then_6, AnimInstance=Self-Reference(), XDotProductCoefficient=1, YDotProductCoefficient=1, ZDotProductCoefficient=1)
   > Set X Movement(X Movement:bool=OR Boolean(A="float > float"(A=@n5, B=0.5), B="float < float"(A=@n5, B=-0.5))) -> (Output_Get:bool)
   n6 = Branch(execute=@n1.then_9, Condition=Get bIsCrouching())
   > n7 = PlayerMovement(AnimInstance=Self-Reference(), XDotProductCoefficient=1, YDotProductCoefficient=1, ZDotProductCoefficient=1)
   > Set PlayerMovement(PlayerMovement:double=MapRangeClamped(Value=@n7, InRangeA=0, InRangeB=-1)) -> (Output_Get:double)
   Set PlayerMovement(execute=@n6.else, PlayerMovement=0)
   Set head Rot(
    execute=@n1.then_7
    head Rot:double=GetAlphaFromReferenceTransformDelta(
     SkeletalMeshComponent=Get Body Mesh()
     BoneName=upperarm_l
     BaseName=spine_05
     DeltaComponentType=EulerY
     InputRange=(X=0,Y=40)
     TwistAxis="1,0,0"
    )
   ) -> (Output_Get:double)
   n8 = GetOwnerSkeletalMeshComponent()
   n9 = BreakRotator(InRot=BreakTransform(InTransform=GetDeltaTransformFromRefPose(self=@n8, BoneName=thigh_r, BaseName=pelvis)).Rotation)
   Set thigh_rot_up_r(execute=@n1.then_8, thigh_rot_up_r:double=MapRangeClamped(Value=@n9.Yaw, InRangeA=-15, InRangeB=-100)) -> (Output_Get:double)
   > Set thigh_rot_inr_r(thigh_rot_inr_r:double=MapRangeClamped(Value=@n9.Pitch, InRangeA=0, InRangeB=-10, OutRangeA=1, OutRangeB=0)) -> (Output_Get:double)
   > Set thigh_rot_otr_r(thigh_rot_otr_r:double=MapRangeClamped(Value=@n9.Pitch, InRangeA=20, InRangeB=50)) -> (Output_Get:double)
   n10 = GetOwnerSkeletalMeshComponent()
   n11 = BreakRotator(InRot=BreakTransform(InTransform=GetDeltaTransformFromRefPose(self=@n10, BoneName=thigh_l, BaseName=pelvis)).Rotation)
   > Set thigh_rot_up_l(thigh_rot_up_l:double=MapRangeClamped(Value=@n11.Yaw, InRangeA=-10, InRangeB=-100)) -> (Output_Get:double)
   > Set thigh_rot_inr_l(thigh_rot_inr_l:double=MapRangeClamped(Value=@n11.Pitch, InRangeA=0, InRangeB=-10, OutRangeA=1, OutRangeB=0)) -> (Output_Get:double)
   > Set thigh_rot_otr_l(thigh_rot_otr_l:double=MapRangeClamped(Value=@n11.Pitch, InRangeA=20, InRangeB=50)) -> (Output_Get:double)
   n12 = Branch(Condition=OR Boolean(A=Get bIsCrouchSprinting(), B=Get X Movement() -> (X Movement:bool)))
   > Set new Is crouching(new Is crouching=false)
   Set new Is crouching(execute=@n12.else, new Is crouching=true)
  ]}
]
variables [
 Body Mesh {type=SkeletalMeshComponent}
 Gravity Override {type=Vector}
 Gravity Override Params {
  type=GravityOverrideParamsStruct
  default {
   skydive_FloatingDown {gravityOverride=<0 -350 300>, windAmplitude=<100 -200 200>, windFrequency=<1.5 3 4>, windOffset=<0 -300 200>}
   skydive_GoingUp {gravityOverride=<0 0 0>}
  }
 }
 head Rot {type=number}
 isCrouching {type=number}
 isNotCrouching {type=number}
 isSkydiving {type=number}
 JumpingFalling {type=boolean}
 new Is crouching {type=boolean}
 PlayerMovement {type=number}
 thigh L Rot {type=number}
 Thigh R Rot {type=number}
 thigh_rot_inr_l {type=number}
 thigh_rot_inr_r {type=number}
 thigh_rot_otr_l {type=number}
 thigh_rot_otr_r {type=number}
 thigh_rot_up_l {type=number}
 thigh_rot_up_r {type=number}
 X Movement {type=boolean}
]
