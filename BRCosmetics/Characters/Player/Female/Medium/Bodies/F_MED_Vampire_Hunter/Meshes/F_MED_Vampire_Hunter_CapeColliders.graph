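# Physics asset of colliders for the Vampire Hunter medium female body cosmetic, named for its cape.
# Capsules cover the pelvis, spine, clavicles, head, feet and every dyn_skirt bone; arm, thigh, calf and neck bodies have no shapes.
# Front, sideA, sideB and back skirt chains on both sides, plus a back-center chain, hang from spine_01.
# Constraints link bodies along these chains and the body skeleton, all with free swing and twist.
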PhysicsAsset /BRCosmetics/Characters/Player/Female/Medium/Bodies/F_MED_Vampire_Hunter/Meshes/F_MED_Vampire_Hunter_CapeColliders
bodies {
 calf_l []
 calf_r []
 clavicle_l [{shapeType=Capsule, length=7.6964, radius=6.3853, center=<4.0147 -0.8354 -1.7419>, rotation=<rot -77.5566 177.3915 7.7314>}]
 clavicle_r [{shapeType=Capsule, length=7.6964, radius=8.2957, center=<-4.0147 0.8354 1.7419>, rotation=<rot -77.5565 177.3915 -172.2685>}]
 dyn_skirt_back_01_c [{
   shapeType=Capsule
   length=6.1707
   radius=3.2739
   shapeName=dyn_skirt_back_01_c_capsule
   center=<6.7545 0 -0.6903>
   rotation=<rot 0 -90 -85.3924>
  }]
 dyn_skirt_back_01_l [{
   shapeType=Capsule
   length=7.3699
   radius=4.4665
   shapeName=dyn_skirt_back_01_l_capsule
   center=<7.862 -0.8486 -0.2026>
   rotation=<rot 0 -83.8202 -87.5411>
  }]
 dyn_skirt_back_01_r [{
   shapeType=Capsule
   length=7.6166
   radius=4.2788
   shapeName=dyn_skirt_back_01_r_capsule
   center=<-8.0524 0.8076 0.2112>
   rotation=<rot 0 -85.4986 -87.346>
  }]
 dyn_skirt_back_02_c [{
   shapeType=Capsule
   length=7.6208
   radius=4.5089
   shapeName=dyn_skirt_back_02_c_capsule
   center=<10.6817 0 -3.3485>
   rotation=<rot 0 -90 -88.2309>
  }]
 dyn_skirt_back_02_l [{
   shapeType=Capsule
   length=6.945
   radius=6.5936
   shapeName=dyn_skirt_back_02_l_capsule
   center=<7.431 -0.5536 -3.9839>
   rotation=<rot 0 -103.526 -87.0365>
  }]
 dyn_skirt_back_02_r [{
   shapeType=Capsule
   length=7.0471
   radius=6.6979
   shapeName=dyn_skirt_back_02_r_capsule
   center=<-7.416 0.4088 4.7161>
   rotation=<rot 0 -105.6808 -86.8923>
  }]
 dyn_skirt_back_03_c [{shapeType=Capsule, length=1.7201, radius=1.2609, shapeName=dyn_skirt_back_03_c_capsule, rotation=<rot 0 -90 0>}]
 dyn_skirt_back_03_l [{
   shapeType=Capsule
   length=7.7249
   radius=5.7086
   shapeName=dyn_skirt_back_03_l_capsule
   center=<7.7676 0.6691 -4.1308>
   rotation=<rot 0 -83.6946 -89.161>
  }]
 dyn_skirt_back_03_r [{
   shapeType=Capsule
   length=7.4795
   radius=5.6582
   shapeName=dyn_skirt_back_03_r_capsule
   center=<-7.5863 -0.6938 4.8777>
   rotation=<rot 0 -84.7729 -89.045>
  }]
 dyn_skirt_back_04_l [{
   shapeType=Capsule
   length=7.3713
   radius=3.4101
   shapeName=dyn_skirt_back_04_l_capsule
   center=<7.5273 -2.3763 -1.3957>
   rotation=<rot 0 -110.0601 -85.1399>
  }]
 dyn_skirt_back_04_r [{
   shapeType=Capsule
   length=7.4295
   radius=3.4142
   shapeName=dyn_skirt_back_04_r_capsule
   center=<-7.4727 2.3634 1.3648>
   rotation=<rot 0 -110.196 -84.9138>
  }]
 dyn_skirt_front_01_l [{
   shapeType=Capsule
   length=5.1054
   radius=3.4798
   shapeName=dyn_skirt_front_01_l_capsule
   center=<6.6187 -0.8081 -0.337>
   rotation=<rot 0 71.2014 -88.6949>
  }]
 dyn_skirt_front_01_r [{
   shapeType=Capsule
   length=5.0897
   radius=3.0739
   shapeName=dyn_skirt_front_01_r_capsule
   center=<-6.7439 0.5185 0.3221>
   rotation=<rot 0 74.0528 -89.5633>
  }]
 dyn_skirt_front_02_l [{
   shapeType=Capsule
   length=4.0237
   radius=2.3835
   shapeName=dyn_skirt_front_02_l_capsule
   center=<6.5737 -0.3171 -0.7522>
   rotation=<rot 0 85.9758 -89.0698>
  }]
 dyn_skirt_front_02_r [{
   shapeType=Capsule
   length=4.2841
   radius=3.009
   shapeName=dyn_skirt_front_02_r_capsule
   center=<-6.6905 0.6311 0.8666>
   rotation=<rot 0 -98.2961 -88.9376>
  }]
 dyn_skirt_front_03_l [{
   shapeType=Capsule
   length=8.9131
   radius=3.1906
   shapeName=dyn_skirt_front_03_l_capsule
   center=<9.9709 -1.5297 -0.4397>
   rotation=<rot 0 87.6271 -89.2575>
  }]
 dyn_skirt_front_03_r [{
   shapeType=Capsule
   length=8.8073
   radius=3.2104
   shapeName=dyn_skirt_front_03_r_capsule
   center=<-9.7422 1.2672 0.6143>
   rotation=<rot 0 -88.5109 -88.348>
  }]
 dyn_skirt_front_04_l [{
   shapeType=Capsule
   length=14.9181
   radius=4.2054
   shapeName=dyn_skirt_front_04_l_capsule
   center=<14.3185 -3.7449 -3.8101>
   rotation=<rot 0 -104.218 -72.7215>
  }]
 dyn_skirt_front_04_r [{
   shapeType=Capsule
   length=14.4792
   radius=4.6637
   shapeName=dyn_skirt_front_04_r_capsule
   center=<-14.6646 4.2495 3.9597>
   rotation=<rot 0 -103.5644 -72.1255>
  }]
 dyn_skirt_sideA_01_l [{
   shapeType=Capsule
   length=6.3962
   radius=6.3361
   shapeName=dyn_skirt_sideA_01_l_capsule
   center=<9.2782 1.3408 -1.0492>
   rotation=<rot 0 -71.1101 -86.4735>
  }]
 dyn_skirt_sideA_01_r [{
   shapeType=Capsule
   length=6.5411
   radius=6.5779
   shapeName=dyn_skirt_sideA_01_r_capsule
   center=<-9.2684 -1.4576 0.8277>
   rotation=<rot 0 -69.1102 -85.2758>
  }]
 dyn_skirt_sideA_02_l [{
   shapeType=Capsule
   length=9.1985
   radius=8.2795
   shapeName=dyn_skirt_sideA_02_l_capsule
   center=<9.9044 2.3515 -3.0154>
   rotation=<rot 0 -76.9856 -89.2118>
  }]
 dyn_skirt_sideA_02_r [{
   shapeType=Capsule
   length=9.0751
   radius=8.3328
   shapeName=dyn_skirt_sideA_02_r_capsule
   center=<-9.6853 -2.3154 4.7266>
   rotation=<rot 0 -76.3136 -88.8767>
  }]
 dyn_skirt_sideA_03_l [{
   shapeType=Capsule
   length=7.5566
   radius=6.2495
   shapeName=dyn_skirt_sideA_03_l_capsule
   center=<7.5482 0.8412 -1.8459>
   rotation=<rot 0 -86.8209 -86.5216>
  }]
 dyn_skirt_sideA_03_r [{
   shapeType=Capsule
   length=7.6271
   radius=6.4894
   shapeName=dyn_skirt_sideA_03_r_capsule
   center=<-7.4264 -0.7115 1.6812>
   rotation=<rot 0 -84.2105 -86.3036>
  }]
 dyn_skirt_sideA_04_l [{
   shapeType=Capsule
   length=8.5122
   radius=4.3534
   shapeName=dyn_skirt_sideA_04_l_capsule
   center=<8.6109 -1.2279 -0.2648>
   rotation=<rot 0 86.6137 -88.163>
  }]
 dyn_skirt_sideA_04_r [{
   shapeType=Capsule
   length=8.4749
   radius=4.3665
   shapeName=dyn_skirt_sideA_04_r_capsule
   center=<-8.6355 1.1 0.2607>
   rotation=<rot 0 87.5043 -88.1279>
  }]
 dyn_skirt_sideB_01_l [{
   shapeType=Capsule
   length=6.9281
   radius=5.451
   shapeName=dyn_skirt_sideB_01_l_capsule
   center=<8.322 -0.2644 0.1666>
   rotation=<rot 0 -87.8906 -89.4081>
  }]
 dyn_skirt_sideB_01_r [{
   shapeType=Capsule
   length=6.9021
   radius=5.475
   shapeName=dyn_skirt_sideB_01_r_capsule
   center=<-8.3108 0.5426 -0.1684>
   rotation=<rot 0 -88.5648 -89.472>
  }]
 dyn_skirt_sideB_02_l [{
   shapeType=Capsule
   length=5.8838
   radius=5.5575
   shapeName=dyn_skirt_sideB_02_l_capsule
   center=<7.7636 0.218 0.2859>
   rotation=<rot 0 85.3937 -89.453>
  }]
 dyn_skirt_sideB_02_r [{
   shapeType=Capsule
   length=5.8837
   radius=5.3482
   shapeName=dyn_skirt_sideB_02_r_capsule
   center=<-7.7461 -0.0115 -0.2859>
   rotation=<rot 0 85.3814 -89.4545>
  }]
 dyn_skirt_sideB_03_l [{
   shapeType=Capsule
   length=15.6351
   radius=5.9541
   shapeName=dyn_skirt_sideB_03_l_capsule
   center=<11.0858 -0.0239 -4.4021>
   rotation=<rot 0 94.4819 -86.8591>
  }]
 dyn_skirt_sideB_03_r [{
   shapeType=Capsule
   length=17.8875
   radius=6.5097
   shapeName=dyn_skirt_sideB_03_r_capsule
   center=<-10.5149 0.0658 2.3739>
   rotation=<rot 0 94.3605 -86.8843>
  }]
 dyn_skirt_sideB_04_l [{
   shapeType=Capsule
   length=10.8801
   radius=5.273
   shapeName=dyn_skirt_sideB_04_l_capsule
   center=<14.0101 -2.0061 -3.7956>
   rotation=<rot 0 -101.7107 -85.9246>
  }]
 dyn_skirt_sideB_04_r [{
   shapeType=Capsule
   length=11.2806
   radius=4.8773
   shapeName=dyn_skirt_sideB_04_r_capsule
   center=<-14.2122 2.0816 0.7254>
   rotation=<rot 0 -101.882 -85.8459>
  }]
 foot_l [{shapeType=Capsule, length=12.5982, radius=6.5127, center=<-2.2072 6.0534 -0.3367>, rotation=<rot 0.0388 1.4701 -89.3982>}]
 foot_r [{shapeType=Capsule, length=12.5982, radius=6.5127, center=<2.2072 -6.0534 0.3367>, rotation=<rot 0.0388 1.4701 90.6018>}]
 head [{shapeType=Capsule, length=0, radius=18, center=<9.0697 0.0746 0>, rotation=<rot 0 -90 0>}]
 lowerarm_l []
 lowerarm_r []
 neck_02 []
 pelvis [{shapeType=Capsule, length=9.4028, radius=10.6596, center=<-1.9544 -2.2029 -0.1981>}]
 spine_01 [{shapeType=Capsule, length=9.0171, radius=9.7354, center=<-2.8183 -1.0873 0>}]
 spine_02 [
  {shapeType=Capsule, length=0, radius=13.4141, center=<1.179 2.2974 0>}
  {shapeType=Capsule, length=20.6875, radius=9.3125, center=<-1.8317 2.606 0>, rotation=<rot 90 0 0>}
 ]
 spine_03 [
  {shapeType=Capsule, length=0.6348, radius=13.8829, center=<-0.0747 2.3725 0>}
  {shapeType=Capsule, length=17.9984, radius=12.0015, center=<-1.2326 2.6059 0>, rotation=<rot 90 0 0>}
 ]
 spine_04 [
  {shapeType=Capsule, length=3.0538, radius=13.1674, center=<5.5182 1.4112 0>}
  {shapeType=Capsule, length=10.0312, radius=9.2812, center=<-1.2326 1.308 0>, rotation=<rot 90 0 0>}
 ]
 spine_05 [{shapeType=Capsule, length=10.9859, radius=10.0791, center=<5.6377 0.3227 0>}]
 thigh_l []
 thigh_r []
 upperarm_l []
 upperarm_r []
}
constraints [
 {
  bone1Name=spine_01
  bone2Name=pelvis
  swing1LimitDegrees=45
  swing1Motion=Free
  swing2LimitDegrees=45
  swing2Motion=Free
  twistLimitDegrees=45
  twistMotion=Free
 }
 {
  bone1Name=spine_02
  bone2Name=spine_01
  swing1LimitDegrees=45
  swing1Motion=Free
  swing2LimitDegrees=45
  swing2Motion=Free
  twistLimitDegrees=45
  twistMotion=Free
 }
 {
  bone1Name=spine_03
  bone2Name=spine_02
  swing1LimitDegrees=45
  swing1Motion=Free
  swing2LimitDegrees=45
  swing2Motion=Free
  twistLimitDegrees=45
  twistMotion=Free
 }
 {
  bone1Name=spine_04
  bone2Name=spine_03
  swing1LimitDegrees=45
  swing1Motion=Free
  swing2LimitDegrees=45
  swing2Motion=Free
  twistLimitDegrees=45
  twistMotion=Free
 }
 {
  bone1Name=spine_05
  bone2Name=spine_04
  swing1LimitDegrees=45
  swing1Motion=Free
  swing2LimitDegrees=45
  swing2Motion=Free
  twistLimitDegrees=45
  twistMotion=Free
 }
 {
  bone1Name=clavicle_l
  bone2Name=spine_05
  swing1LimitDegrees=45
  swing1Motion=Free
  swing2LimitDegrees=45
  swing2Motion=Free
  twistLimitDegrees=45
  twistMotion=Free
 }
 {
  bone1Name=upperarm_l
  bone2Name=clavicle_l
  swing1LimitDegrees=45
  swing1Motion=Free
  swing2LimitDegrees=45
  swing2Motion=Free
  twistLimitDegrees=45
  twistMotion=Free
 }
 {
  bone1Name=lowerarm_l
  bone2Name=upperarm_l
  swing1LimitDegrees=45
  swing1Motion=Free
  swing2LimitDegrees=45
  swing2Motion=Free
  twistLimitDegrees=45
  twistMotion=Free
 }
 {
  bone1Name=neck_02
  bone2Name=spine_05
  swing1LimitDegrees=45
  swing1Motion=Free
  swing2LimitDegrees=45
  swing2Motion=Free
  twistLimitDegrees=45
  twistMotion=Free
 }
 {
  bone1Name=head
  bone2Name=neck_02
  swing1LimitDegrees=45
  swing1Motion=Free
  swing2LimitDegrees=45
  swing2Motion=Free
  twistLimitDegrees=45
  twistMotion=Free
 }
 {
  bone1Name=calf_l
  bone2Name=thigh_l
  swing1LimitDegrees=45
  swing1Motion=Free
  swing2LimitDegrees=45
  swing2Motion=Free
  twistLimitDegrees=45
  twistMotion=Free
 }
 {
  bone1Name=thigh_l
  bone2Name=pelvis
  swing1LimitDegrees=45
  swing1Motion=Free
  swing2LimitDegrees=45
  swing2Motion=Free
  twistLimitDegrees=45
  twistMotion=Free
 }
 {
  bone1Name=foot_l
  bone2Name=calf_l
  swing1LimitDegrees=45
  swing1Motion=Free
  swing2LimitDegrees=45
  swing2Motion=Free
  twistLimitDegrees=45
  twistMotion=Free
 }
 {
  bone1Name=clavicle_r
  bone2Name=spine_05
  swing1LimitDegrees=45
  swing1Motion=Free
  swing2LimitDegrees=45
  swing2Motion=Free
  twistLimitDegrees=45
  twistMotion=Free
 }
 {
  bone1Name=lowerarm_r
  bone2Name=upperarm_r
  swing1LimitDegrees=45
  swing1Motion=Free
  swing2LimitDegrees=45
  swing2Motion=Free
  twistLimitDegrees=45
  twistMotion=Free
 }
 {
  bone1Name=thigh_r
  bone2Name=pelvis
  swing1LimitDegrees=45
  swing1Motion=Free
  swing2LimitDegrees=45
  swing2Motion=Free
  twistLimitDegrees=45
  twistMotion=Free
 }
 {
  bone1Name=calf_r
  bone2Name=thigh_r
  swing1LimitDegrees=45
  swing1Motion=Free
  swing2LimitDegrees=45
  swing2Motion=Free
  twistLimitDegrees=45
  twistMotion=Free
 }
 {
  bone1Name=foot_r
  bone2Name=calf_r
  swing1LimitDegrees=45
  swing1Motion=Free
  swing2LimitDegrees=45
  swing2Motion=Free
  twistLimitDegrees=45
  twistMotion=Free
 }
 {
  bone1Name=upperarm_r
  bone2Name=clavicle_r
  swing1LimitDegrees=45
  swing1Motion=Free
  swing2LimitDegrees=45
  swing2Motion=Free
  twistLimitDegrees=45
  twistMotion=Free
 }
 {
  bone1Name=dyn_skirt_sideA_01_l
  bone2Name=spine_01
  swing1LimitDegrees=45
  swing1Motion=Free
  swing2LimitDegrees=45
  swing2Motion=Free
  twistLimitDegrees=45
  twistMotion=Free
 }
 {
  bone1Name=dyn_skirt_sideA_02_l
  bone2Name=dyn_skirt_sideA_01_l
  swing1LimitDegrees=45
  swing1Motion=Free
  swing2LimitDegrees=45
  swing2Motion=Free
  twistLimitDegrees=45
  twistMotion=Free
 }
 {
  bone1Name=dyn_skirt_sideA_03_l
  bone2Name=dyn_skirt_sideA_02_l
  swing1LimitDegrees=45
  swing1Motion=Free
  swing2LimitDegrees=45
  swing2Motion=Free
  twistLimitDegrees=45
  twistMotion=Free
 }
 {
  bone1Name=dyn_skirt_sideA_04_l
  bone2Name=dyn_skirt_sideA_03_l
  swing1LimitDegrees=45
  swing1Motion=Free
  swing2LimitDegrees=45
  swing2Motion=Free
  twistLimitDegrees=45
  twistMotion=Free
 }
 {
  bone1Name=dyn_skirt_back_01_c
  bone2Name=spine_01
  swing1LimitDegrees=45
  swing1Motion=Free
  swing2LimitDegrees=45
  swing2Motion=Free
  twistLimitDegrees=45
  twistMotion=Free
 }
 {
  bone1Name=dyn_skirt_back_02_c
  bone2Name=dyn_skirt_back_01_c
  swing1LimitDegrees=45
  swing1Motion=Free
  swing2LimitDegrees=45
  swing2Motion=Free
  twistLimitDegrees=45
  twistMotion=Free
 }
 {
  bone1Name=dyn_skirt_back_03_c
  bone2Name=dyn_skirt_back_02_c
  swing1LimitDegrees=45
  swing1Motion=Free
  swing2LimitDegrees=45
  swing2Motion=Free
  twistLimitDegrees=45
  twistMotion=Free
 }
 {
  bone1Name=dyn_skirt_sideB_01_l
  bone2Name=spine_01
  swing1LimitDegrees=45
  swing1Motion=Free
  swing2LimitDegrees=45
  swing2Motion=Free
  twistLimitDegrees=45
  twistMotion=Free
 }
 {
  bone1Name=dyn_skirt_sideB_02_l
  bone2Name=dyn_skirt_sideB_01_l
  swing1LimitDegrees=45
  swing1Motion=Free
  swing2LimitDegrees=45
  swing2Motion=Free
  twistLimitDegrees=45
  twistMotion=Free
 }
 {
  bone1Name=dyn_skirt_sideB_03_l
  bone2Name=dyn_skirt_sideB_02_l
  swing1LimitDegrees=45
  swing1Motion=Free
  swing2LimitDegrees=45
  swing2Motion=Free
  twistLimitDegrees=45
  twistMotion=Free
 }
 {
  bone1Name=dyn_skirt_sideB_04_l
  bone2Name=dyn_skirt_sideB_03_l
  swing1LimitDegrees=45
  swing1Motion=Free
  swing2LimitDegrees=45
  swing2Motion=Free
  twistLimitDegrees=45
  twistMotion=Free
 }
 {
  bone1Name=dyn_skirt_front_01_l
  bone2Name=spine_01
  swing1LimitDegrees=45
  swing1Motion=Free
  swing2LimitDegrees=45
  swing2Motion=Free
  twistLimitDegrees=45
  twistMotion=Free
 }
 {
  bone1Name=dyn_skirt_front_02_l
  bone2Name=dyn_skirt_front_01_l
  swing1LimitDegrees=45
  swing1Motion=Free
  swing2LimitDegrees=45
  swing2Motion=Free
  twistLimitDegrees=45
  twistMotion=Free
 }
 {
  bone1Name=dyn_skirt_front_03_l
  bone2Name=dyn_skirt_front_02_l
  swing1LimitDegrees=45
  swing1Motion=Free
  swing2LimitDegrees=45
  swing2Motion=Free
  twistLimitDegrees=45
  twistMotion=Free
 }
 {
  bone1Name=dyn_skirt_front_04_l
  bone2Name=dyn_skirt_front_03_l
  swing1LimitDegrees=45
  swing1Motion=Free
  swing2LimitDegrees=45
  swing2Motion=Free
  twistLimitDegrees=45
  twistMotion=Free
 }
 {
  bone1Name=dyn_skirt_back_01_l
  bone2Name=spine_01
  swing1LimitDegrees=45
  swing1Motion=Free
  swing2LimitDegrees=45
  swing2Motion=Free
  twistLimitDegrees=45
  twistMotion=Free
 }
 {
  bone1Name=dyn_skirt_back_02_l
  bone2Name=dyn_skirt_back_01_l
  swing1LimitDegrees=45
  swing1Motion=Free
  swing2LimitDegrees=45
  swing2Motion=Free
  twistLimitDegrees=45
  twistMotion=Free
 }
 {
  bone1Name=dyn_skirt_back_03_l
  bone2Name=dyn_skirt_back_02_l
  swing1LimitDegrees=45
  swing1Motion=Free
  swing2LimitDegrees=45
  swing2Motion=Free
  twistLimitDegrees=45
  twistMotion=Free
 }
 {
  bone1Name=dyn_skirt_back_04_l
  bone2Name=dyn_skirt_back_03_l
  swing1LimitDegrees=45
  swing1Motion=Free
  swing2LimitDegrees=45
  swing2Motion=Free
  twistLimitDegrees=45
  twistMotion=Free
 }
 {
  bone1Name=dyn_skirt_sideA_01_r
  bone2Name=spine_01
  swing1LimitDegrees=45
  swing1Motion=Free
  swing2LimitDegrees=45
  swing2Motion=Free
  twistLimitDegrees=45
  twistMotion=Free
 }
 {
  bone1Name=dyn_skirt_sideA_02_r
  bone2Name=dyn_skirt_sideA_01_r
  swing1LimitDegrees=45
  swing1Motion=Free
  swing2LimitDegrees=45
  swing2Motion=Free
  twistLimitDegrees=45
  twistMotion=Free
 }
 {
  bone1Name=dyn_skirt_sideA_03_r
  bone2Name=dyn_skirt_sideA_02_r
  swing1LimitDegrees=45
  swing1Motion=Free
  swing2LimitDegrees=45
  swing2Motion=Free
  twistLimitDegrees=45
  twistMotion=Free
 }
 {
  bone1Name=dyn_skirt_sideA_04_r
  bone2Name=dyn_skirt_sideA_03_r
  swing1LimitDegrees=45
  swing1Motion=Free
  swing2LimitDegrees=45
  swing2Motion=Free
  twistLimitDegrees=45
  twistMotion=Free
 }
 {
  bone1Name=dyn_skirt_sideB_01_r
  bone2Name=spine_01
  swing1LimitDegrees=45
  swing1Motion=Free
  swing2LimitDegrees=45
  swing2Motion=Free
  twistLimitDegrees=45
  twistMotion=Free
 }
 {
  bone1Name=dyn_skirt_sideB_02_r
  bone2Name=dyn_skirt_sideB_01_r
  swing1LimitDegrees=45
  swing1Motion=Free
  swing2LimitDegrees=45
  swing2Motion=Free
  twistLimitDegrees=45
  twistMotion=Free
 }
 {
  bone1Name=dyn_skirt_sideB_03_r
  bone2Name=dyn_skirt_sideB_02_r
  swing1LimitDegrees=45
  swing1Motion=Free
  swing2LimitDegrees=45
  swing2Motion=Free
  twistLimitDegrees=45
  twistMotion=Free
 }
 {
  bone1Name=dyn_skirt_sideB_04_r
  bone2Name=dyn_skirt_sideB_03_r
  swing1LimitDegrees=45
  swing1Motion=Free
  swing2LimitDegrees=45
  swing2Motion=Free
  twistLimitDegrees=45
  twistMotion=Free
 }
 {
  bone1Name=dyn_skirt_back_01_r
  bone2Name=spine_01
  swing1LimitDegrees=45
  swing1Motion=Free
  swing2LimitDegrees=45
  swing2Motion=Free
  twistLimitDegrees=45
  twistMotion=Free
 }
 {
  bone1Name=dyn_skirt_back_02_r
  bone2Name=dyn_skirt_back_01_r
  swing1LimitDegrees=45
  swing1Motion=Free
  swing2LimitDegrees=45
  swing2Motion=Free
  twistLimitDegrees=45
  twistMotion=Free
 }
 {
  bone1Name=dyn_skirt_back_03_r
  bone2Name=dyn_skirt_back_02_r
  swing1LimitDegrees=45
  swing1Motion=Free
  swing2LimitDegrees=45
  swing2Motion=Free
  twistLimitDegrees=45
  twistMotion=Free
 }
 {
  bone1Name=dyn_skirt_back_04_r
  bone2Name=dyn_skirt_back_03_r
  swing1LimitDegrees=45
  swing1Motion=Free
  swing2LimitDegrees=45
  swing2Motion=Free
  twistLimitDegrees=45
  twistMotion=Free
 }
 {
  bone1Name=dyn_skirt_front_01_r
  bone2Name=spine_01
  swing1LimitDegrees=45
  swing1Motion=Free
  swing2LimitDegrees=45
  swing2Motion=Free
  twistLimitDegrees=45
  twistMotion=Free
 }
 {
  bone1Name=dyn_skirt_front_02_r
  bone2Name=dyn_skirt_front_01_r
  swing1LimitDegrees=45
  swing1Motion=Free
  swing2LimitDegrees=45
  swing2Motion=Free
  twistLimitDegrees=45
  twistMotion=Free
 }
 {
  bone1Name=dyn_skirt_front_03_r
  bone2Name=dyn_skirt_front_02_r
  swing1LimitDegrees=45
  swing1Motion=Free
  swing2LimitDegrees=45
  swing2Motion=Free
  twistLimitDegrees=45
  twistMotion=Free
 }
 {
  bone1Name=dyn_skirt_front_04_r
  bone2Name=dyn_skirt_front_03_r
  swing1LimitDegrees=45
  swing1Motion=Free
  swing2LimitDegrees=45
  swing2Motion=Free
  twistLimitDegrees=45
  twistMotion=Free
 }
]
