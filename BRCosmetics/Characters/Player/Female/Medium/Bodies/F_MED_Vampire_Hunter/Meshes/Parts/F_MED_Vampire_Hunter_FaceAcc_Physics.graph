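# Physics asset for the face accessory part of the female medium Vampire Hunter cosmetic, covering its dynamic hair and hood bones.
# Capsule bodies cover the hair and hood bones plus head, upper spine, left clavicle and upper arm, with several large offset capsules on the head.
# Hood and upper hair chains hang off the head, while two lower hair chains hang off spine_05, all with narrow swing and zero twist limits.
# The left arm joints move freely.
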PhysicsAsset /BRCosmetics/Characters/Player/Female/Medium/Bodies/F_MED_Vampire_Hunter/Meshes/Parts/F_MED_Vampire_Hunter_FaceAcc_Physics
bodies {
 clavicle_l [{shapeType=Capsule, length=13.25, radius=5, shapeName=clavicle_l_capsule, rotation=<rot 90 0 0>}]
 dyn_hair_l_01 [{shapeType=Capsule, length=3, radius=1, shapeName=dyn_hair_l_01_capsule, center=<2.5252 0 0>, rotation=<rot 90 0 0>}]
 dyn_hair_l_02 [{shapeType=Capsule, length=3, radius=1, shapeName=dyn_hair_l_02_capsule, center=<2.5252 0 0>, rotation=<rot 90 0 0>}]
 dyn_hair_l_03 [{shapeType=Capsule, length=2.25, radius=2.5, shapeName=dyn_hair_l_03_capsule, center=<2.5147 0 0>, rotation=<rot 90 0 0>}]
 dyn_hair_l_04 [{shapeType=Capsule, length=5.1304, radius=1.5277, shapeName=dyn_hair_l_04_capsule, center=<4.528 0 0>, rotation=<rot 90 0 0>}]
 dyn_hair_r_01 [{shapeType=Capsule, length=4.5126, radius=1, shapeName=dyn_hair_r_01_capsule, center=<3.8825 0 0>, rotation=<rot 90 0 0>}]
 dyn_hair_r_02 [{shapeType=Capsule, length=2.146, radius=1.1267, shapeName=dyn_hair_r_02_capsule, center=<2.5252 0 0>, rotation=<rot 90 0 0>}]
 dyn_hair_r_03 [{shapeType=Capsule, length=3, radius=1, shapeName=dyn_hair_r_03_capsule, center=<2.5252 0 0>, rotation=<rot 90 0 0>}]
 dyn_hair_r_04 [{shapeType=Capsule, length=3, radius=1, shapeName=dyn_hair_r_04_capsule, center=<2.5252 0 0>, rotation=<rot 90 0 0>}]
 dyn_hair_r_05 [{shapeType=Capsule, length=1.25, radius=2.5, shapeName=dyn_hair_r_05_capsule, center=<2.7967 0 0>, rotation=<rot 90 0 0>}]
 dyn_hair_r_06 [{
   shapeType=Capsule
   length=3.9321
   radius=2.316
   shapeName=dyn_hair_r_06_capsule
   center=<2.6 1.5 0>
   rotation=<rot 64.6332 -162.6049 165.5957>
  }]
 dyn_hood_01 [{shapeType=Capsule, length=2.0447, radius=3.3854, shapeName=dyn_hood_01_capsule, center=<0.8061 0 0>}]
 dyn_hood_side_r_01 [{
   shapeType=Capsule
   length=6.75
   radius=1.75
   shapeName=dyn_hood_side_r_01_capsule
   center=<0.7893 0.0521 0>
   rotation=<rot 83.1915 93.7787 180>
  }]
 head [
  {shapeType=Capsule, length=9.323, radius=7.1015, center=<1.3017 0 0>, rotation=<rot 90 0 0>}
  {shapeType=Capsule, length=15, radius=15, shapeName=head_capsule, center=<3.9545 7.3246 -22.5796>, rotation=<rot 76.1794 -174.5892 168.9198>}
  {shapeType=Capsule, length=15, radius=15, shapeName=head_capsule, center=<3.768 7.2682 22.7121>, rotation=<rot 79.5096 8.1504 -10.418>}
  {shapeType=Capsule, length=22.2459, radius=7.7541, shapeName=head_capsule, center=<0.1105 -4.7404 0.0469>, rotation=<rot 0 -0.0005 -5.9161>}
 ]
 spine_05 [{shapeType=Capsule, length=17.9154, radius=10, center=<-2.4453 1.5841 0>}]
 upperarm_l [{shapeType=Capsule, length=14.005, radius=5.755, shapeName=upperarm_l_capsule, center=<6.9916 0 0>, rotation=<rot 90 0 180>}]
}
constraints [
 {bone1Name=head, bone2Name=spine_05, swing1LimitDegrees=45, swing2LimitDegrees=45, twistLimitDegrees=45}
 {bone1Name=dyn_hood_01, bone2Name=head, swing1LimitDegrees=10, swing2LimitDegrees=10, twistLimitDegrees=0}
 {bone1Name=dyn_hair_r_01, bone2Name=head, swing1LimitDegrees=3, swing2LimitDegrees=20, twistLimitDegrees=0}
 {bone1Name=dyn_hair_r_02, bone2Name=dyn_hair_r_01, swing1LimitDegrees=20, swing2LimitDegrees=20, twistLimitDegrees=0}
 {bone1Name=dyn_hair_r_04, bone2Name=dyn_hair_r_02, swing1LimitDegrees=20, swing2LimitDegrees=20, twistLimitDegrees=0}
 {bone1Name=dyn_hair_r_03, bone2Name=dyn_hair_r_01, swing1LimitDegrees=20, swing2LimitDegrees=20, twistLimitDegrees=0}
 {bone1Name=dyn_hair_l_01, bone2Name=head, swing1LimitDegrees=3, swing2LimitDegrees=15, twistLimitDegrees=0}
 {bone1Name=dyn_hair_l_02, bone2Name=dyn_hair_l_01, swing1LimitDegrees=3, swing2LimitDegrees=20, twistLimitDegrees=0}
 {bone1Name=dyn_hair_r_05, bone2Name=spine_05, swing1LimitDegrees=10, swing2LimitDegrees=10, twistLimitDegrees=0}
 {bone1Name=dyn_hair_r_06, bone2Name=dyn_hair_r_05, swing1LimitDegrees=10, swing2LimitDegrees=10, twistLimitDegrees=0}
 {bone1Name=dyn_hair_l_03, bone2Name=spine_05, swing1LimitDegrees=10, swing2LimitDegrees=10, twistLimitDegrees=0}
 {bone1Name=dyn_hair_l_04, bone2Name=dyn_hair_l_03, swing1LimitDegrees=10, swing2LimitDegrees=10, twistLimitDegrees=0}
 {
  bone1Name=upperarm_l
  bone2Name=clavicle_l
  swing1LimitDegrees=45
  swing1Motion=Free
  swing2LimitDegrees=45
  swing2Motion=Free
  twistLimitDegrees=45
  twistMotion=Free
 }
 {
  bone1Name=clavicle_l
  bone2Name=spine_05
  swing1LimitDegrees=45
  swing1Motion=Free
  swing2LimitDegrees=45
  swing2Motion=Free
  twistLimitDegrees=45
  twistMotion=Free
 }
 {bone1Name=dyn_hood_side_r_01, bone2Name=head, swing1LimitDegrees=10, swing2LimitDegrees=10, twistLimitDegrees=0}
]
