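# Animation blueprint for the Female Medium Vampire Hunter face accessory mesh, adding secondary motion and corrective offsets to hair and hood bones.
# Copied body poses are blended per bone, then passed through a Control Rig borrowed from Bistro Astronaut and a RigidBody simulation.
# Additive bone offsets follow left-arm rotation for hair and head rotation for the hood, the latter only with certain weapon types.
# The final hood constraint has zero alpha, and the targeting-driven move-hair flag feeds no pose node.
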AnimBlueprint /BRCosmetics/Characters/Player/Female/Medium/Bodies/F_MED_Vampire_Hunter/Meshes/Parts/F_MED_Vampire_Hunter_FaceAcc_AnimBP
target_skeleton=&/F_MED_Vampire_Hunter_FaceAcc_Skeleton
graphs [
 AnimGraph {nodes [
   n1 = Layered blend per bone(
    BasePose=Copy Pose From Mesh() {bCopyCurves=true, bUseAttachedParent=true}
    BlendPoses_0=Copy Pose From Mesh(SourceMeshComponent=Get mesh() -> (mesh:Skeletal Mesh Component Object Reference))
   ) {layerSetup [{branchFilters [
       {boneName=upperarm_l}
       {boneName=upperarm_r}
       {boneName=lowerarm_l}
       {boneName=lowerarm_r}
       {boneName=thigh_l}
       {boneName=thigh_r}
       {boneName=calf_l}
       {boneName=calf_r}
      ]}]}
   n2 = Control Rig(Source=@n1) {
    alphaBoolBlend {blendOption=Linear}
    controlRigAssetReference {blueprintRigClass=/BRCosmetics/Characters/Player/Female/Medium/Bodies/F_MED_Bistro_Astronaut/Meshes/Parts/F_MED_Bistro_Astronaut_FaceAcc_CtrlRig.F_MED_Bistro_Astronaut_FaceAcc_CtrlRig_C}
   }
   n3 = RigidBody(ComponentPose=Local To Component(LocalPose=@n2)) {
    bClampLinearTranslationLimitToRefPose=true
    bOverrideWorldGravity=true
    overridePhysicsAsset=&/F_MED_Vampire_Hunter_FaceAcc_Physics
    simulationSpace=BaseBoneSpace
    alphaBoolBlend {blendOption=Linear}
    baseBoneRef {boneName=root}
    componentLinearAccScale=<0.01 0.01 0.01>
    componentLinearVelScale=<0.01 0.01 0.01>
    overrideWorldGravity {z=-980}
   }
   n4 = "Transform (Modify) Bone - Bone: dyn_hair_l_03"(ComponentPose=@n3, Alpha=Get arm L Rot() -> (arm L Rot:double), Translation="-2.5,-0.485,-0.947", Rotation="0,0,0", Scale="0.8,0.8,1") {
    rotationMode=BMM_Additive
    scaleMode=BMM_Additive
    translationMode=BMM_Additive
    alphaBoolBlend {blendOption=Linear}
    boneToModify {boneName=dyn_hair_l_03}
    scale=<0.8 0.8>
    translation=<-2.5 -0.485 -0.947>
   }
   n5 = "Transform (Modify) Bone - Bone: dyn_hood_side_r_01"(
    ComponentPose=@n4
    Alpha=Get head Rot() -> (head Rot:double)
    Translation="-0.000,-3.806,0.000"
    Rotation="-0.000,0.000,20.523"
    Scale="1,1,1"
   ) {
    rotationMode=BMM_Additive
    scaleMode=BMM_Additive
    translationMode=BMM_Additive
    alphaBoolBlend {blendOption=Linear}
    boneToModify {boneName=dyn_hood_side_r_01}
    rotation {roll=20.523}
    translation {y=-3.806}
   }
   n6 = "Constraint - dyn_hood_side_r_01 (P)"(ComponentPose=@n5, Alpha=0, ConstraintWeights_0=1) {
    alphaBoolBlend {blendOption=Linear}
    boneToModify {boneName=dyn_hood_side_r_01}
    constraintSetup [{targetBone {boneName=spine_05}}]
    constraintWeights [1]
   }
   Output Pose(Result=Component To Local(ComponentPose=@n6)) {layerGroup=None}
  ]}
 EventGraph {nodes [
   Event BlueprintUpdateAnimation()
   > n1 = Sequence!(execute) -> (then_0, then_1, then_2:Exec, then_3:Exec, then_4:Exec, then_5:Exec)
   > Set arm L Rot(arm L Rot:double=GetAlphaFromReferenceTransformDelta(
     SkeletalMeshComponent=Get mesh()
     BoneName=upperarm_l
     BaseName=spine_05
     DeltaComponentType=EulerY
     InputRange=(X=0,Y=40)
     TwistAxis="1,0,0"
    )) -> (Output_Get:double)
   Set mesh(execute=@n1.then_1, mesh:Skeletal Mesh Component Object Reference=GetOwnerSkeletalMeshComponent()) -> (Output_Get:Skeletal Mesh Component Object Reference)
   n2 = Branch(execute=@n1.then_2, Condition=AND Boolean(A=Get Alt_weapon() -> (Alt_weapon:bool), B=Get new is targeting() -> (new is targeting:bool)))
   > Set move hair(move hair:bool=true) -> (Output_Get:bool)
   Set move hair(execute=@n2.else, move hair=false)
   n3 = Cast To FortPlayerAnimInstance(execute=@n1.then_3, Object=GetAnimInstance(self=Get mesh()))
   > Set Alt_weapon(Alt_weapon:bool=false) -> (Output_Get:bool)
   > n4 = Switch on EFortWeaponCoreAnimation!(execute, Selection:EFortWeaponCoreAnimation Enum=Get WeaponCoreAnim(self=@n3)) -> (
    Melee:Exec
    Pistol:Exec
    Shotgun:Exec
    PaperBlueprint:Exec
    Rifle:Exec
    MeleeOneHand:Exec
    MachinePistol:Exec
    RocketLauncher:Exec
    GrenadeLauncher:Exec
    GoingCommando:Exec
    AssaultRifle:Exec
    TacticalShotgun:Exec
    SniperRifle:Exec
    TrapPlacement:Exec
    ShoulderLauncher:Exec
    AbilityDecoTool:Exec
    Crossbow:Exec
    C4:Exec
    RemoteControl:Exec
    DualWield:Exec
    AR_BullPup:Exec
    AR_ForwardGrip:Exec
    MedPackPaddles:Exec
    SMG_P90:Exec
    AR_DrumGun:Exec
    Consumable_Small:Exec
    Consumable_Large:Exec
    Balloon:Exec
    MountedTurret:Exec
    CreativeTool:Exec
    ExplosiveBow:Exec
    AshtonIndigo:Exec
    AshtonChicago:Exec
    MeleeDualWield:Exec
    Unarmed:Exec
   )
   Branch(execute=[@n4.DualWield, @n4.MachinePistol, @n4.Pistol, @n4.Rifle, @n4.Shotgun, @n4.TacticalShotgun])
   > Set Alt_weapon(Alt_weapon=true)
   n5 = Branch(execute=@n1.then_4, Condition=Get Alt_weapon())
   > Set head Rot(head Rot:double=GetAlphaFromReferenceTransformDelta(SkeletalMeshComponent=Get mesh(), BoneName=head, BaseName=spine_05, InputRange=(X=-35,Y=0), OutputRange=(X=1,Y=0), TwistAxis="1,0,0")) -> (Output_Get:double)
   Set head Rot(execute=@n5.else, head Rot=0)
   Event OnBeginTargeting() {is=title K2Node_Event}
   > Set new is targeting(new is targeting:bool=true) -> (Output_Get:bool)
   Event OnEndTargeting() {is=title K2Node_Event}
   > Set new is targeting(new is targeting=false)
   PlayerMovement(XDotProductCoefficient=1, YDotProductCoefficient=1, ZDotProductCoefficient=1)
  ]}
]
variables [
 Alt_weapon {type=boolean}
 arm L Rot {type=number}
 body Mesh {type=SkeletalMeshComponent}
 head Rot {type=number}
 Is Targeting {type=boolean}
 mesh {type=SkeletalMeshComponent}
 move hair {type=boolean}
 new is targeting {type=boolean}
]
